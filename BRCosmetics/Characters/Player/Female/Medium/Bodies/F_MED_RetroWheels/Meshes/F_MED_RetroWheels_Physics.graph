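# Physics asset for the F_MED_RetroWheels female medium body, rigging dynamic cloth-volume bones with simple boxes and capsules.
# Mirrored upper chains (Up01, Up02, A, A01) hang from spine_03, and lower B chains hang from the pelvis through free-angular joints.
# Only the left B chain branches into C01 and C02 spheres, modelled as zero-length capsules.
# Pelvis, spine and thigh-twist capsules cover the body, with the thighs left unconstrained.
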PhysicsAsset /BRCosmetics/Characters/Player/Female/Medium/Bodies/F_MED_RetroWheels/Meshes/F_MED_RetroWheels_Physics
bodies {
 dyn_Cloth_volume_l_A [{
   shapeType=Box
   extentX=12.219
   extentY=5.6035
   extentZ=17.2571
   shapeName=dyn_Cloth_volume_l_A_box
   center=<-2.9888 0.9413 1.2123>
   rotation=<rot 0.0001 -9.9996 15>
  }]
 dyn_Cloth_volume_l_A01 [{
   shapeType=Box
   extentX=12.219
   extentY=4.8244
   extentZ=17.2571
   shapeName=dyn_Cloth_volume_l_A01_box
   center=<-2.9888 0.9413 1.2123>
   rotation=<rot 0.0001 -9.9996 15>
  }]
 dyn_Cloth_volume_l_B [{
   shapeType=Capsule
   length=0
   radius=0.5
   shapeName=dyn_Cloth_volume_l_B_capsule
   center=<-6.702 -3.4792 0.1533>
   rotation=<rot 8.5507 -90 12.6093>
  }]
 dyn_Cloth_volume_l_B01 [{
   shapeType=Capsule
   length=9.6093
   radius=4.2819
   shapeName=dyn_Cloth_volume_l_B01_capsule
   center=<-8.0277 -3.906 4.252>
   rotation=<rot 7.9804 -90 -6.7696>
  }]
 dyn_Cloth_volume_l_B02 [{
   shapeType=Capsule
   length=10.5479
   radius=4.5336
   shapeName=dyn_Cloth_volume_l_B02_capsule
   center=<-3.0802 -2.2217 2.9359>
   rotation=<rot 9.4188 -90 -4.8053>
  }]
 dyn_Cloth_volume_l_C01 [{shapeType=Capsule, length=0, radius=4.0409, shapeName=dyn_Cloth_volume_l_C01_capsule}]
 dyn_Cloth_volume_l_C02 [{shapeType=Capsule, length=0, radius=4.0409, shapeName=dyn_Cloth_volume_l_C02_capsule}]
 dyn_Cloth_volume_l_Up01 [{
   shapeType=Capsule
   length=0
   radius=0.5
   shapeName=dyn_Cloth_volume_l_Up01_capsule
   center=<-7.8863 -0.9135 -1.8958>
   rotation=<rot -4.0188 -90 9.7848>
  }]
 dyn_Cloth_volume_l_Up02 [{shapeType=Capsule, length=8.2254, radius=3.9499, shapeName=dyn_Cloth_volume_l_Up02_capsule, center=<-6.2977 0 0>}]
 dyn_Cloth_volume_r_A [{
   shapeType=Box
   extentX=12.219
   extentY=3.8618
   extentZ=17.2571
   shapeName=dyn_Cloth_volume_l_A_box
   center=<-3.049 0.5999 -2.1899>
   rotation=<rot 0.0001 -9.9995 -5>
  }]
 dyn_Cloth_volume_r_A01 [{
   shapeType=Box
   extentX=12.219
   extentY=3.8618
   extentZ=17.2571
   shapeName=dyn_Cloth_volume_l_A01_box
   center=<-3.049 0.5999 -2.1899>
   rotation=<rot 0.0001 -9.9995 -5>
  }]
 dyn_Cloth_volume_r_B [{
   shapeType=Capsule
   length=0
   radius=0.5
   shapeName=dyn_Cloth_volume_r_B_capsule
   center=<-6.6162 -2.7376 0.2519>
   rotation=<rot -7.3148 -90 -9.5686>
  }]
 dyn_Cloth_volume_r_B01 [{
   shapeType=Capsule
   length=10.9673
   radius=4.1342
   shapeName=dyn_Cloth_volume_r_B01_capsule
   center=<-8.8913 -3.1298 -1.9984>
   rotation=<rot -9.3863 -90 9.0135>
  }]
 dyn_Cloth_volume_r_B02 [{
   shapeType=Capsule
   length=12.7477
   radius=4.294
   shapeName=dyn_Cloth_volume_r_B02_capsule
   center=<-3.117 -0.3611 -0.626>
   rotation=<rot -9.5174 -90 3.3022>
  }]
 dyn_Cloth_volume_r_Up01 [{
   shapeType=Capsule
   length=0
   radius=0.5
   shapeName=dyn_Cloth_volume_r_Up01_capsule
   center=<-7.6162 -0.9 1.6762>
   rotation=<rot -0.0566 -90 0.6148>
  }]
 dyn_Cloth_volume_r_Up02 [{shapeType=Capsule, length=8.2254, radius=3.9499, shapeName=dyn_Cloth_volume_r_Up02_capsule, center=<-6.2977 0 0>}]
 pelvis [{shapeType=Capsule, length=9.6041, radius=13.3938, shapeName=pelvis_capsule, center=<-4.749 -0.7056 0>}]
 spine_02 [{shapeType=Capsule, length=6.4969, radius=12.0714, shapeName=spine_02_capsule}]
 spine_03 [{shapeType=Capsule, length=0, radius=0.5, shapeName=spine_03_capsule, center=<-0.8849 -0.3904 0.0252>, rotation=<rot 0.1422 -90 0.1368>}]
 thigh_twist_01_l [{
   shapeType=Capsule
   length=33.8767
   radius=8.9778
   shapeName=thigh_l_capsule
   center=<10.5108 -1.1125 0.6247>
   rotation=<rot 90 6.7214 6.7215>
  }]
 thigh_twist_01_r [{
   shapeType=Capsule
   length=34.6642
   radius=9.0329
   shapeName=thigh_l_capsule
   center=<-10.9871 1.0323 0.999>
   rotation=<rot 90 179.9992 179.9993>
  }]
}
constraints [
 {bone1Name=spine_02, bone2Name=pelvis, swing1LimitDegrees=45, swing2LimitDegrees=45, twistLimitDegrees=45}
 {bone1Name=dyn_Cloth_volume_l_A01, bone2Name=dyn_Cloth_volume_l_A, swing1LimitDegrees=45, swing2LimitDegrees=45, twistLimitDegrees=45}
 {bone1Name=dyn_Cloth_volume_l_A, bone2Name=dyn_Cloth_volume_l_Up02, swing1LimitDegrees=45, swing2LimitDegrees=45, twistLimitDegrees=45}
 {bone1Name=dyn_Cloth_volume_r_A01, bone2Name=dyn_Cloth_volume_r_A, swing1LimitDegrees=45, swing2LimitDegrees=45, twistLimitDegrees=45}
 {bone1Name=dyn_Cloth_volume_r_A, bone2Name=dyn_Cloth_volume_r_Up02, swing1LimitDegrees=45, swing2LimitDegrees=45, twistLimitDegrees=45}
 {bone1Name=dyn_Cloth_volume_r_B02, bone2Name=dyn_Cloth_volume_r_B01, swing1LimitDegrees=45, swing2LimitDegrees=45, twistLimitDegrees=15}
 {bone1Name=dyn_Cloth_volume_l_B02, bone2Name=dyn_Cloth_volume_l_B01, swing1LimitDegrees=45, swing2LimitDegrees=45, twistLimitDegrees=15}
 {bone1Name=dyn_Cloth_volume_l_B01, bone2Name=dyn_Cloth_volume_l_B, swing1LimitDegrees=45, swing2LimitDegrees=45, twistLimitDegrees=15}
 {
  bone1Name=dyn_Cloth_volume_l_B
  bone2Name=pelvis
  swing1LimitDegrees=45
  swing1Motion=Free
  swing2LimitDegrees=45
  swing2Motion=Free
  twistLimitDegrees=45
  twistMotion=Free
 }
 {bone1Name=dyn_Cloth_volume_r_B01, bone2Name=dyn_Cloth_volume_r_B, swing1LimitDegrees=45, swing2LimitDegrees=45, twistLimitDegrees=15}
 {
  bone1Name=dyn_Cloth_volume_r_B
  bone2Name=pelvis
  swing1LimitDegrees=45
  swing1Motion=Free
  swing2LimitDegrees=45
  swing2Motion=Free
  twistLimitDegrees=45
  twistMotion=Free
 }
 {bone1Name=dyn_Cloth_volume_l_Up02, bone2Name=dyn_Cloth_volume_l_Up01, swing1LimitDegrees=45, swing2LimitDegrees=45, twistLimitDegrees=45}
 {bone1Name=dyn_Cloth_volume_l_Up01, bone2Name=spine_03, swing1LimitDegrees=45, swing2LimitDegrees=45, twistLimitDegrees=45}
 {
  bone1Name=spine_03
  bone2Name=spine_02
  swing1LimitDegrees=45
  swing1Motion=Free
  swing2LimitDegrees=45
  swing2Motion=Free
  twistLimitDegrees=45
  twistMotion=Free
 }
 {bone1Name=dyn_Cloth_volume_r_Up02, bone2Name=dyn_Cloth_volume_r_Up01, swing1LimitDegrees=45, swing2LimitDegrees=45, twistLimitDegrees=45}
 {bone1Name=dyn_Cloth_volume_r_Up01, bone2Name=spine_03, swing1LimitDegrees=45, swing2LimitDegrees=45, twistLimitDegrees=45}
 {bone1Name=dyn_Cloth_volume_l_B01, bone2Name=dyn_Cloth_volume_l_C01, swing1LimitDegrees=45, swing2LimitDegrees=45, twistLimitDegrees=15}
 {bone1Name=dyn_Cloth_volume_l_B02, bone2Name=dyn_Cloth_volume_l_C02, swing1LimitDegrees=45, swing2LimitDegrees=45, twistLimitDegrees=15}
]
