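# Animation Blueprint for the F_MED_RetroWheels female cosmetic body, adding secondary motion and corrective posing to skirt cloth, collar, shoulder and button bones.
# A crouch/slide Control Rig, alpha-driven bone offsets and constraints shape cached poses.
# AnimDynamics spring chains and RigidBody simulation, enabled by is_WearingCape, add cloth motion.
# Movement, crouch, dive and DBNO states blend these poses before multi-bone weighted constraints place the buttons.
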AnimBlueprint /BRCosmetics/Characters/Player/Female/Medium/Bodies/F_MED_RetroWheels/Meshes/F_MED_RetroWheels_AnimBP
target_skeleton=&/F_MED_RetroWheels_Skeleton
graphs [
 AnimGraph {nodes [
   n1 = Anim Dynamics - Bone: dyn_Cloth_volume_r_Up01!(ComponentPose, Alpha=1, GravityOverride:Vector="0,0,-980", ExternalForce="0,0,0") -> (Pose) {
    angularDampingOverride=0.99
    angularSpringConstant=800
    bAngularSpring=true
    bOverrideAngularDamping=true
    bUseGravityOverride=true
    gravityScale=0.3
    simulationSpace=BoneRelative
    alphaBoolBlend {blendOption=Linear}
    boundBone {boneName=dyn_Cloth_volume_r_Up01}
    componentAppliedLinearAccClamp=<5000 5000 5000>
    componentLinearAccScale=<0.45 0.45 0.65>
    componentLinearVelScale=<0.55 0.55 0.7>
    gravityOverride {z=-980}
    physicsBodyDefinitions [{
      collisionType=CustomSphere
      sphereCollisionRadius=2
      boundBone {boneName=dyn_Cloth_volume_r_Up01}
      boxExtents=<30 30 30>
      constraintSetup {angularLimitsMax=<1 1 20>, angularLimitsMin=<-1 -1 -2>, angularTarget=<0 0 0>}
      localJointOffset=<-5 0.2 1>
     }]
    relativeSpaceBone {boneName=gravity}
   }
   Anim Dynamics - Bone: dyn_Cloth_volume_r_Up02!(ComponentPose=@n1, Alpha=1, GravityOverride:Vector="0,0,-980", ExternalForce="0,0,0") -> (Pose) {
    angularDampingOverride=0.99
    angularSpringConstant=800
    bAngularSpring=true
    bOverrideAngularDamping=true
    bUseGravityOverride=true
    gravityScale=0.3
    simulationSpace=BoneRelative
    alphaBoolBlend {blendOption=Linear}
    boundBone {boneName=dyn_Cloth_volume_r_Up02}
    componentAppliedLinearAccClamp=<5000 5000 5000>
    componentLinearAccScale=<0.45 0.45 0.65>
    componentLinearVelScale=<0.55 0.55 0.7>
    gravityOverride {z=-980}
    physicsBodyDefinitions [{
      collisionType=CustomSphere
      sphereCollisionRadius=2
      boundBone {boneName=dyn_Cloth_volume_r_Up02}
      boxExtents=<20 20 20>
      constraintSetup {angularLimitsMax=<1 1 20>, angularLimitsMin=<-1 -1 -2>, angularTarget=<0 0 0>}
      localJointOffset=<-5 0.2 1>
     }]
    relativeSpaceBone {boneName=gravity}
   }
   Anim Dynamics - Bone: dyn_Cloth_volume_l_Up01!(ComponentPose, Alpha=1, GravityOverride:Vector="0,0,500", ExternalForce="0,0,0") -> (Pose) {
    angularDampingOverride=0.99
    angularSpringConstant=800
    bAngularSpring=true
    bOverrideAngularDamping=true
    bUseGravityOverride=true
    gravityScale=0.3
    simulationSpace=BoneRelative
    alphaBoolBlend {blendOption=Linear}
    boundBone {boneName=dyn_Cloth_volume_l_Up01}
    componentAppliedLinearAccClamp=<5000 5000 5000>
    componentLinearAccScale=<0.45 0.45 0.65>
    componentLinearVelScale=<0.55 0.55 0.7>
    gravityOverride {z=500}
    physicsBodyDefinitions [{
      collisionType=CustomSphere
      sphereCollisionRadius=2
      boundBone {boneName=dyn_Cloth_volume_l_Up01}
      boxExtents=<30 30 30>
      constraintSetup {angularLimitsMax=<0.1 0.1 20>, angularLimitsMin=<-0.1 -0.1 -2>, angularTarget=<0 0 0>}
      localJointOffset=<-5 0.2 1>
     }]
    relativeSpaceBone {boneName=gravity}
   }
   Anim Dynamics - Bone: dyn_Cloth_volume_l_Up02!(ComponentPose, Alpha=1, GravityOverride:Vector="0,0,500", ExternalForce="0,0,0") -> (Pose) {
    angularDampingOverride=0.99
    angularSpringConstant=800
    bAngularSpring=true
    bOverrideAngularDamping=true
    bUseGravityOverride=true
    gravityScale=0.3
    simulationSpace=BoneRelative
    alphaBoolBlend {blendOption=Linear}
    boundBone {boneName=dyn_Cloth_volume_l_Up02}
    componentAppliedLinearAccClamp=<5000 5000 5000>
    componentLinearAccScale=<0.45 0.45 0.65>
    componentLinearVelScale=<0.55 0.55 0.7>
    gravityOverride {z=500}
    physicsBodyDefinitions [{
      collisionType=CustomSphere
      sphereCollisionRadius=2
      boundBone {boneName=dyn_Cloth_volume_l_Up02}
      boxExtents=<20 20 20>
      constraintSetup {angularLimitsMax=<0.1 0.1 20>, angularLimitsMin=<-0.1 -0.1 -2>, angularTarget=<0 0 0>}
      localJointOffset=<-5 0.2 1>
     }]
    relativeSpaceBone {boneName=gravity}
   }
   Component To Local(ComponentPose=Local To Component())
   n2 = "Constraint - dyn_Cloth_volume_l_C01 (PP)"!(
    ComponentPose=Local To Component(LocalPose=Use cached pose 'SkirtCachePose'())
    Alpha=1
    ConstraintWeights_0=0.85
    ConstraintWeights_1:float=0.15
   ) -> (Pose) {
    alphaBoolBlend {blendOption=Linear}
    boneToModify {boneName=dyn_Cloth_volume_l_C01}
    constraintSetup [{targetBone {boneName=pelvis}}, {targetBone {boneName=dyn_Cloth_volume_l_B}}]
    constraintWeights [1,1]
   }
   n3 = "Constraint - dyn_Cloth_volume_r_C01 (PP)"!(ComponentPose=@n2, Alpha=1, ConstraintWeights_0=0.85, ConstraintWeights_1:float=0.15) -> (Pose) {
    alphaBoolBlend {blendOption=Linear}
    boneToModify {boneName=dyn_Cloth_volume_r_C01}
    constraintSetup [{targetBone {boneName=pelvis}}, {targetBone {boneName=dyn_Cloth_volume_r_B}}]
    constraintWeights [1,1]
   }
   n4 = "Transform (Modify) Bone - Bone: dyn_Cloth_volume_l_C01"(ComponentPose=@n3, Alpha=1, Translation="0,0,0", Rotation="0,20,0", Scale="1,1,1") {
    rotationMode=BMM_Additive
    rotationSpace=BCS_BoneSpace
    scaleMode=BMM_Additive
    scaleSpace=BCS_BoneSpace
    translationMode=BMM_Additive
    translationSpace=BCS_BoneSpace
    alphaBoolBlend {blendOption=Linear}
    boneToModify {boneName=dyn_Cloth_volume_l_C01}
    rotation {yaw=20}
   }
   n5 = "Transform (Modify) Bone - Bone: dyn_Cloth_volume_r_C01"(ComponentPose=@n4, Alpha=1, Translation="0,0,0", Rotation="0,20,0", Scale="1,1,1") {
    rotationMode=BMM_Additive
    rotationSpace=BCS_BoneSpace
    scaleMode=BMM_Additive
    scaleSpace=BCS_BoneSpace
    translationMode=BMM_Additive
    translationSpace=BCS_BoneSpace
    alphaBoolBlend {blendOption=Linear}
    boneToModify {boneName=dyn_Cloth_volume_r_C01}
    rotation {yaw=20}
   }
   n6 = "Constraint - dyn_Cloth_volume_bk (PP)"!(ComponentPose=@n5, Alpha=1, ConstraintWeights_0=0.5, ConstraintWeights_1:float=0.5) -> (Pose) {
    alphaBoolBlend {blendOption=Linear}
    boneToModify {boneName=dyn_Cloth_volume_bk}
    constraintSetup [{targetBone {boneName=dyn_Cloth_volume_l_C01}}, {targetBone {boneName=dyn_Cloth_volume_r_C01}}]
    constraintWeights [1,1]
   }
   n7 = "Transform (Modify) Bone - Bone: dyn_Cloth_volume_l_C01"(
    ComponentPose=Local To Component(LocalPose=Use cached pose 'SkirtCachePose'())
    Alpha=1
    Translation="0,0,0"
    Rotation="0,-15,0"
    Scale="1,1,1"
   ) {
    rotationMode=BMM_Additive
    rotationSpace=BCS_BoneSpace
    scaleMode=BMM_Additive
    scaleSpace=BCS_BoneSpace
    translationMode=BMM_Additive
    translationSpace=BCS_BoneSpace
    alphaBoolBlend {blendOption=Linear}
    boneToModify {boneName=dyn_Cloth_volume_l_C01}
    rotation {yaw=-15}
   }
   n8 = "Transform (Modify) Bone - Bone: dyn_Cloth_volume_r_C01"(ComponentPose=@n7, Alpha=1, Translation="0,0,0", Rotation="0,0,0", Scale="1,1,1") {
    rotationMode=BMM_Additive
    rotationSpace=BCS_BoneSpace
    scaleMode=BMM_Additive
    scaleSpace=BCS_BoneSpace
    translationMode=BMM_Additive
    translationSpace=BCS_BoneSpace
    alphaBoolBlend {blendOption=Linear}
    boneToModify {boneName=dyn_Cloth_volume_r_C01}
   }
   n9 = "Constraint - dyn_Cloth_volume_l_C01 (PP)"(ComponentPose=@n8, Alpha=1, ConstraintWeights_0=0.1, ConstraintWeights_1=0.9) {
    alphaBoolBlend {blendOption=Linear}
    boneToModify {boneName=dyn_Cloth_volume_l_C01}
    constraintSetup [{targetBone {boneName=thigh_l}}, {targetBone {boneName=dyn_Cloth_volume_l_C}}]
    constraintWeights [1,1]
   }
   n10 = "Constraint - dyn_Cloth_volume_bk01 (PP)"!(ComponentPose=@n9, Alpha=1, ConstraintWeights_0=0.5, ConstraintWeights_1:float=0.5) -> (Pose) {
    alphaBoolBlend {blendOption=Linear}
    boneToModify {boneName=dyn_Cloth_volume_bk01}
    constraintSetup [{targetBone {boneName=dyn_Cloth_volume_l_C01}}, {targetBone {boneName=dyn_Cloth_volume_r_C01}}]
    constraintWeights [1,1]
   }
   n11 = "Constraint - dyn_Cloth_volume_r_C01 (PP)"(ComponentPose=@n10, Alpha=1, ConstraintWeights_0=0.1, ConstraintWeights_1=0.9) {
    alphaBoolBlend {blendOption=Linear}
    boneToModify {boneName=dyn_Cloth_volume_r_C01}
    constraintSetup [{targetBone {boneName=thigh_r}}, {targetBone {boneName=dyn_Cloth_volume_r_C}}]
    constraintWeights [1,1]
   }
   n12 = "Transform (Modify) Bone - Bone: dyn_Cloth_volume_r_A"(ComponentPose=@n11, Alpha=1, Translation="-5,0,0", Rotation="0,0,0", Scale="1,1,1") {
    rotationMode=BMM_Additive
    rotationSpace=BCS_BoneSpace
    scaleMode=BMM_Additive
    scaleSpace=BCS_BoneSpace
    translationMode=BMM_Additive
    translationSpace=BCS_BoneSpace
    alphaBoolBlend {blendOption=Linear}
    boneToModify {boneName=dyn_Cloth_volume_r_A}
    translation {x=-5}
   }
   n13 = "Transform (Modify) Bone - Bone: dyn_Cloth_volume_l_A"(ComponentPose=@n12, Alpha=1, Translation="0,0,1", Rotation="-15,-10,10", Scale="1,1,1") {
    rotationMode=BMM_Additive
    rotationSpace=BCS_BoneSpace
    scaleMode=BMM_Additive
    scaleSpace=BCS_BoneSpace
    translationMode=BMM_Additive
    translationSpace=BCS_BoneSpace
    alphaBoolBlend {blendOption=Linear}
    boneToModify {boneName=dyn_Cloth_volume_l_A}
    rotation=<rot -15 -10 10>
    translation {z=1}
   }
   n14 = Blend Poses by bool(
    BlendPose_0=Component To Local(ComponentPose=@n6)
    BlendPose_1=Component To Local(ComponentPose=@n13)
    BlendTime_0=0.2
    BlendTime_1=0.2
    bActiveValue=Get MovingAnd NotCrouchNotDivingNotEmote() -> (MovingAnd NotCrouchNotDivingNotEmote:bool)
   ) {blendType=Linear}
   n15 = Use cached pose 'SkirtCachePose'()
   n16 = Use cached pose 'SkirtDynPose'()
   n17 = "Constraint - dyn_Cloth_volume_l_B (PP)"!(
    ComponentPose=Local To Component(LocalPose=Two Way Blend(A=@n15, B=@n16, Alpha=0.5) {blendNode {alphaBoolBlend {blendOption=Linear}}})
    Alpha=1
    ConstraintWeights_0=0.5
    ConstraintWeights_1:float=0.5
   ) -> (Pose) {
    alphaBoolBlend {blendOption=Linear}
    boneToModify {boneName=dyn_Cloth_volume_l_B}
    constraintSetup [{targetBone {boneName=dyn_Cloth_volume_l_C}}, {targetBone {boneName=dyn_Cloth_volume_l_A}}]
    constraintWeights [1,1]
   }
   n18 = "Constraint - dyn_Cloth_volume_r_B (PP)"!(ComponentPose=@n17, Alpha=1, ConstraintWeights_0=0.5, ConstraintWeights_1:float=0.5) -> (Pose) {
    alphaBoolBlend {blendOption=Linear}
    boneToModify {boneName=dyn_Cloth_volume_r_B}
    constraintSetup [{targetBone {boneName=dyn_Cloth_volume_r_C}}, {targetBone {boneName=dyn_Cloth_volume_r_A}}]
    constraintWeights [1,1]
   }
   n19 = Component To Local(ComponentPose=@n18)
   n20 = Use cached pose 'DBNOPose'()
   n21 = Get DBNO_Alpha() -> (DBNO_Alpha:double)
   n22 = "Transform (Modify) Bone - Bone: dyn_Cloth_volume_l_C"(
    ComponentPose=Local To Component(LocalPose=Two Way Blend(A=@n20, B=@n19, Alpha=@n21) {blendNode {alphaBoolBlend {blendOption=Linear}}})
    Alpha=1
    Translation="-2,0,0"
    Rotation="0,0,0"
    Scale="1,1,1"
   ) {
    rotationMode=BMM_Additive
    rotationSpace=BCS_BoneSpace
    scaleMode=BMM_Additive
    scaleSpace=BCS_BoneSpace
    translationMode=BMM_Additive
    translationSpace=BCS_BoneSpace
    alphaBoolBlend {blendOption=Linear}
    boneToModify {boneName=dyn_Cloth_volume_l_C}
    translation {x=-2}
   }
   n23 = "Transform (Modify) Bone - Bone: dyn_Cloth_volume_r_C"(ComponentPose=@n22, Alpha=1, Translation="-2,0,0", Rotation="0,0,0", Scale="1,1,1") {
    rotationMode=BMM_Additive
    rotationSpace=BCS_BoneSpace
    scaleMode=BMM_Additive
    scaleSpace=BCS_BoneSpace
    translationMode=BMM_Additive
    translationSpace=BCS_BoneSpace
    alphaBoolBlend {blendOption=Linear}
    boneToModify {boneName=dyn_Cloth_volume_r_C}
    translation {x=-2}
   }
   n24 = "Transform (Modify) Bone - Bone: dyn_Cloth_volume_bk"(ComponentPose=@n23, Alpha=1, Translation="-2,0,0", Rotation="0,0,0", Scale="1,1,1") {
    rotationMode=BMM_Additive
    rotationSpace=BCS_BoneSpace
    scaleMode=BMM_Additive
    scaleSpace=BCS_BoneSpace
    translationMode=BMM_Additive
    translationSpace=BCS_BoneSpace
    alphaBoolBlend {blendOption=Linear}
    boneToModify {boneName=dyn_Cloth_volume_bk}
    translation {x=-2}
   }
   n25 = Blend Poses by bool(
    BlendPose_0=Component To Local(ComponentPose=@n24)
    BlendPose_1=Use cached pose 'SkirtDynPose'()
    bActiveValue=Get DivingOrDBNO() -> (DivingOrDBNO:bool)
   ) {blendType=Linear}
   n26 = "Constraint - dyn_buttons_01 (PPPPPPP)"!(
    ComponentPose=Local To Component(LocalPose=Blend Poses by bool(BlendPose_0=@n25, BlendPose_1=@n14) {blendType=Linear})
    Alpha=1
    ConstraintWeights_0=0.4454
    ConstraintWeights_1:float=0.247
    ConstraintWeights_2:float=0.1975
    ConstraintWeights_3:float=0.0927
    ConstraintWeights_4:float=0.0099
    ConstraintWeights_5:float=0.0071
    ConstraintWeights_6:float=0.0003
   ) -> (Pose) {
    alphaBoolBlend {blendOption=Linear}
    boneToModify {boneName=dyn_buttons_01}
    constraintSetup [
     {targetBone {boneName=spine_04}}
     {targetBone {boneName=spine_05}}
     {targetBone {boneName=clavicle_l}}
     {targetBone {boneName=deform_upperarm_fr_l}}
     {targetBone {boneName=upperarm_twist_01_l}}
     {targetBone {boneName=upperarmDn_l}}
     {targetBone {boneName=deform_upperarm_up_l}}
    ]
    constraintWeights [1,1,1,1,1,1,1]
   }
   n27 = "Constraint - dyn_buttons_02 (PPPPPP)"!(
    ComponentPose=@n26
    Alpha=1
    ConstraintWeights_0=0.8733
    ConstraintWeights_1:float=0.048
    ConstraintWeights_2:float=0.0309
    ConstraintWeights_3:float=0.0297
    ConstraintWeights_4:float=0.0111
    ConstraintWeights_5:float=0.0069
   ) -> (Pose) {
    alphaBoolBlend {blendOption=Linear}
    boneToModify {boneName=dyn_buttons_02}
    constraintSetup [
     {targetBone {boneName=spine_04}}
     {targetBone {boneName=spine_03}}
     {targetBone {boneName=dyn_Cloth_volume_l_Up01}}
     {targetBone {boneName=clavicle_l}}
     {targetBone {boneName=spine_05}}
     {targetBone {boneName=deform_upperarm_fr_l}}
    ]
    constraintWeights [1,1,1,1,1,1]
   }
   n28 = "Constraint - dyn_buttons_03 (PPPP)"!(
    ComponentPose=@n27
    Alpha=1
    ConstraintWeights_0=0.3636
    ConstraintWeights_1:float=0.2524
    ConstraintWeights_2:float=0.217
    ConstraintWeights_3:float=0.1669
   ) -> (Pose) {
    alphaBoolBlend {blendOption=Linear}
    boneToModify {boneName=dyn_buttons_03}
    constraintSetup [
     {targetBone {boneName=dyn_Cloth_volume_l_Up01}}
     {targetBone {boneName=spine_04}}
     {targetBone {boneName=dyn_Cloth_volume_l_Up02}}
     {targetBone {boneName=spine_03}}
    ]
    constraintWeights [1,1,1,1]
   }
   n29 = "Constraint - dyn_buttons_04 (PPPPPPP)"!(
    ComponentPose=@n28
    Alpha=1
    ConstraintWeights_0=0.6041
    ConstraintWeights_1:float=0.2592
    ConstraintWeights_2:float=0.0949
    ConstraintWeights_3:float=0.0271
    ConstraintWeights_4:float=0.0116
    ConstraintWeights_5:float=0.0016
    ConstraintWeights_6:float=0.0015
   ) -> (Pose) {
    alphaBoolBlend {blendOption=Linear}
    boneToModify {boneName=dyn_buttons_04}
    constraintSetup [
     {targetBone {boneName=dyn_Cloth_volume_l_Up02}}
     {targetBone {boneName=dyn_Cloth_volume_l_A}}
     {targetBone {boneName=dyn_Cloth_volume_l_Up01}}
     {targetBone {boneName=dyn_Cloth_volume_l_A01}}
     {targetBone {boneName=spine_03}}
     {targetBone {boneName=dyn_Cloth_volume_l_B}}
     {targetBone {boneName=spine_04}}
    ]
    constraintWeights [1,1,1,1,1,1,1]
   }
   Output Pose(Result=Component To Local(ComponentPose=@n29)) {layerGroup=None}
   n30 = Control Rig!(Source=Use cached pose 'bodyPose'(), Alpha, isCrouching:bool=Get bIsCrouching(), isSliding:bool=Get isSliding() -> (isSliding:bool)) -> (Pose) {
    alphaBoolBlend {blendOption=Linear}
    controlRigAssetReference {blueprintRigClass=&/F_MED_RetroWheels_Crouhing_CtrlRig.F_MED_RetroWheels_Crouhing_CtrlRig_C}
   }
   n31 = "Constraint - upperarmDn_l (PP)"!(ComponentPose=Local To Component(LocalPose=@n30), Alpha=1, ConstraintWeights_0=0.3, ConstraintWeights_1:float=0.7) -> (Pose) {
    alphaBoolBlend {blendOption=Linear}
    boneToModify {boneName=upperarmDn_l}
    constraintSetup [{targetBone {boneName=spine_05}}, {targetBone {boneName=upperarm_twist_01_l}}]
    constraintWeights [1,1]
   }
   n32 = "Constraint - upperarmDn_r (PP)"!(ComponentPose=@n31, Alpha=1, ConstraintWeights_0=0.3, ConstraintWeights_1:float=0.7) -> (Pose) {
    alphaBoolBlend {blendOption=Linear}
    boneToModify {boneName=upperarmDn_r}
    constraintSetup [{targetBone {boneName=spine_05}}, {targetBone {boneName=upperarm_twist_01_r}}]
    constraintWeights [1,1]
   }
   n33 = "Transform (Modify) Bone - Bone: upperarmDn_l"(ComponentPose=@n32, Alpha=Get Shoulder_rot_dn_l() -> (Shoulder_rot_dn_l:double), Translation="-5,0,5", Rotation="0,0,0", Scale="1,1,1") {
    rotationMode=BMM_Additive
    rotationSpace=BCS_BoneSpace
    scaleMode=BMM_Additive
    scaleSpace=BCS_BoneSpace
    translationMode=BMM_Additive
    translationSpace=BCS_BoneSpace
    alphaBoolBlend {blendOption=Linear}
    boneToModify {boneName=upperarmDn_l}
    translation {x=-5, z=5}
   }
   n34 = "Transform (Modify) Bone - Bone: upperarmDn_r"(ComponentPose=@n33, Alpha=Get Shoulder_rot_dn_r() -> (Shoulder_rot_dn_r:double), Translation="5,0,-5", Rotation="0,0,0", Scale="1,1,1") {
    rotationMode=BMM_Additive
    rotationSpace=BCS_BoneSpace
    scaleMode=BMM_Additive
    scaleSpace=BCS_BoneSpace
    translationMode=BMM_Additive
    translationSpace=BCS_BoneSpace
    alphaBoolBlend {blendOption=Linear}
    boneToModify {boneName=upperarmDn_r}
    translation {x=5, z=-5}
   }
   n35 = Get Shoulder_rot_dn_l_0() -> (Shoulder_rot_dn_l_0:double)
   n36 = "Transform (Modify) Bone - Bone: upperarmDn_l"(ComponentPose=@n34, Alpha=@n35, Translation="5,0,5", Rotation="0,0,0", Scale="1,1,1") {
    rotationMode=BMM_Additive
    rotationSpace=BCS_BoneSpace
    scaleMode=BMM_Additive
    scaleSpace=BCS_BoneSpace
    translationMode=BMM_Additive
    translationSpace=BCS_BoneSpace
    alphaBoolBlend {blendOption=Linear}
    boneToModify {boneName=upperarmDn_l}
    translation {x=5, z=5}
   }
   n37 = Get Shoulder_rot_dn_r_0() -> (Shoulder_rot_dn_r_0:double)
   n38 = "Transform (Modify) Bone - Bone: upperarmDn_r"(ComponentPose=@n36, Alpha=@n37, Translation="-5,0,-5", Rotation="0,0,0", Scale="1,1,1") {
    rotationMode=BMM_Additive
    rotationSpace=BCS_BoneSpace
    scaleMode=BMM_Additive
    scaleSpace=BCS_BoneSpace
    translationMode=BMM_Additive
    translationSpace=BCS_BoneSpace
    alphaBoolBlend {blendOption=Linear}
    boneToModify {boneName=upperarmDn_r}
    translation {x=-5, z=-5}
   }
   n39 = "Transform (Modify) Bone - Bone: dyn_Cloth_volume_l_B"(ComponentPose=@n38, Alpha=@n35, Translation="0,15,0", Rotation="0,10,0", Scale="1,1,1") {
    rotationMode=BMM_Additive
    rotationSpace=BCS_BoneSpace
    scaleMode=BMM_Additive
    scaleSpace=BCS_BoneSpace
    translationMode=BMM_Additive
    translationSpace=BCS_BoneSpace
    alphaBoolBlend {blendOption=Linear}
    boneToModify {boneName=dyn_Cloth_volume_l_B}
    rotation {yaw=10}
    translation {y=15}
   }
   n40 = "Transform (Modify) Bone - Bone: dyn_Cloth_volume_r_B"(ComponentPose=@n39, Alpha=@n37, Translation="0,10,0", Rotation="0,20,0", Scale="1,1,1") {
    rotationMode=BMM_Additive
    rotationSpace=BCS_BoneSpace
    scaleMode=BMM_Additive
    scaleSpace=BCS_BoneSpace
    translationMode=BMM_Additive
    translationSpace=BCS_BoneSpace
    alphaBoolBlend {blendOption=Linear}
    boneToModify {boneName=dyn_Cloth_volume_r_B}
    rotation {yaw=20}
    translation {y=10}
   }
   n41 = "Transform (Modify) Bone - Bone: dyn_Cloth_volume_l_Up02"(ComponentPose=@n40, Alpha=@n35, Translation="0,5,0", Rotation="0,10,0", Scale="1,1,1") {
    rotationMode=BMM_Additive
    rotationSpace=BCS_BoneSpace
    scaleMode=BMM_Additive
    scaleSpace=BCS_BoneSpace
    translationMode=BMM_Additive
    translationSpace=BCS_BoneSpace
    alphaBoolBlend {blendOption=Linear}
    boneToModify {boneName=dyn_Cloth_volume_l_Up02}
    rotation {yaw=10}
    translation {y=5}
   }
   n42 = "Transform (Modify) Bone - Bone: upperarmDn_l"(ComponentPose=@n41, Alpha=Get Shoulder_rot_up_l() -> (Shoulder_rot_up_l:double), Translation="-5,0,-5", Rotation="0,0,0", Scale="1,1,1") {
    rotationMode=BMM_Additive
    rotationSpace=BCS_BoneSpace
    scaleMode=BMM_Additive
    scaleSpace=BCS_BoneSpace
    translationMode=BMM_Additive
    translationSpace=BCS_BoneSpace
    alphaBoolBlend {blendOption=Linear}
    boneToModify {boneName=upperarmDn_l}
    translation {x=-5, z=-5}
   }
   n43 = "Transform (Modify) Bone - Bone: upperarmDn_r"(ComponentPose=@n42, Alpha=Get Shoulder_rot_up_r() -> (Shoulder_rot_up_r:double), Translation="5,5,5", Rotation="0,0,0", Scale="1,1,1") {
    rotationMode=BMM_Additive
    rotationSpace=BCS_BoneSpace
    scaleMode=BMM_Additive
    scaleSpace=BCS_BoneSpace
    translationMode=BMM_Additive
    translationSpace=BCS_BoneSpace
    alphaBoolBlend {blendOption=Linear}
    boneToModify {boneName=upperarmDn_r}
    translation=<5 5 5>
   }
   n44 = "Transform (Modify) Bone - Bone: dyn_collar_l"(ComponentPose=@n43, Alpha=Get upperarm_shape_l() -> (upperarm_shape_l:double), Translation="-5,0,0", Rotation="30,20,0", Scale="1,1,1") {
    rotationMode=BMM_Additive
    rotationSpace=BCS_BoneSpace
    scaleMode=BMM_Additive
    scaleSpace=BCS_BoneSpace
    translationMode=BMM_Additive
    translationSpace=BCS_BoneSpace
    alphaBoolBlend {blendOption=Linear}
    boneToModify {boneName=dyn_collar_l}
    rotation {pitch=30, yaw=20}
    translation {x=-5}
   }
   n45 = "Transform (Modify) Bone - Bone: dyn_collar_r"(ComponentPose=@n44, Alpha=Get upperarm_shape_r() -> (upperarm_shape_r:double), Translation="5,0,0", Rotation="-30,-20,0", Scale="1,1,1") {
    rotationMode=BMM_Additive
    rotationSpace=BCS_BoneSpace
    scaleMode=BMM_Additive
    scaleSpace=BCS_BoneSpace
    translationMode=BMM_Additive
    translationSpace=BCS_BoneSpace
    alphaBoolBlend {blendOption=Linear}
    boneToModify {boneName=dyn_collar_r}
    rotation {pitch=-30, yaw=-20}
    translation {x=5}
   }
   Save cached pose 'slidingPose'(Pose=Component To Local(ComponentPose=@n45)) {cacheName=slidingPose}
   n46 = Get thigh_shape_fn_r_0() -> (thigh_shape_fn_r_0:double)
   n47 = "Transform (Modify) Bone - Bone: dyn_Cloth_volume_l_C"(
    ComponentPose=Local To Component(LocalPose=Use cached pose 'CrouchingPose'())
    Alpha=@n46
    Translation="0,0,0"
    Rotation="0,10,0"
    Scale="1,1,1"
   ) {
    rotationMode=BMM_Additive
    rotationSpace=BCS_BoneSpace
    scaleMode=BMM_Additive
    scaleSpace=BCS_BoneSpace
    translationMode=BMM_Additive
    translationSpace=BCS_BoneSpace
    alphaBoolBlend {blendOption=Linear}
    boneToModify {boneName=dyn_Cloth_volume_l_C}
    rotation {yaw=10}
   }
   n48 = "Transform (Modify) Bone - Bone: dyn_Cloth_volume_l_B"(ComponentPose=@n47, Alpha=@n46, Translation="0,0,0", Rotation="0,5,0", Scale="1,1,1") {
    rotationMode=BMM_Additive
    rotationSpace=BCS_BoneSpace
    scaleMode=BMM_Additive
    scaleSpace=BCS_BoneSpace
    translationMode=BMM_Additive
    translationSpace=BCS_BoneSpace
    alphaBoolBlend {blendOption=Linear}
    boneToModify {boneName=dyn_Cloth_volume_l_B}
    rotation {yaw=5}
   }
   n49 = "Transform (Modify) Bone - Bone: dyn_Cloth_volume_r_C02"(ComponentPose=@n48, Alpha=1, Translation="0,0,0", Rotation="0,5,0", Scale="1,1,1") {
    rotationMode=BMM_Additive
    rotationSpace=BCS_BoneSpace
    scaleMode=BMM_Additive
    scaleSpace=BCS_BoneSpace
    translationMode=BMM_Additive
    translationSpace=BCS_BoneSpace
    alphaBoolBlend {blendOption=Linear}
    boneToModify {boneName=dyn_Cloth_volume_r_C02}
    rotation {yaw=5}
   }
   n50 = "Transform (Modify) Bone - Bone: dyn_Cloth_volume_r_B01"(ComponentPose=@n49, Alpha=1, Translation="0,0,0", Rotation="0,4,0", Scale="1,1,1") {
    rotationMode=BMM_Additive
    rotationSpace=BCS_BoneSpace
    scaleMode=BMM_Additive
    scaleSpace=BCS_BoneSpace
    translationMode=BMM_Additive
    translationSpace=BCS_BoneSpace
    alphaBoolBlend {blendOption=Linear}
    boneToModify {boneName=dyn_Cloth_volume_r_B01}
    rotation {yaw=4}
   }
   n51 = Component To Local(ComponentPose=@n50)
   n52 = Use cached pose 'CrouchingPose'()
   n53 = Get CrouchingAndNotMovingAndNotShape() -> (CrouchingAndNotMovingAndNotShape:bool)
   n54 = "Constraint - dyn_Cloth_volume_bk (PP)"(
    ComponentPose=Local To Component(LocalPose=Blend Poses by bool(BlendPose_0=@n51, BlendPose_1=@n52, bActiveValue=@n53) {blendType=Linear})
    Alpha=1
    ConstraintWeights_0=0.5
    ConstraintWeights_1=0.5
   ) {
    alphaBoolBlend {blendOption=Linear}
    boneToModify {boneName=dyn_Cloth_volume_bk}
    constraintSetup [{targetBone {boneName=dyn_Cloth_volume_l_C}}, {targetBone {boneName=dyn_Cloth_volume_r_C}}]
    constraintWeights [1,1]
   }
   n55 = "Constraint - dyn_Cloth_volume_bk01 (PP)"(ComponentPose=@n54, Alpha=1, ConstraintWeights_0=0.5, ConstraintWeights_1=0.5) {
    alphaBoolBlend {blendOption=Linear}
    boneToModify {boneName=dyn_Cloth_volume_bk01}
    constraintSetup [{targetBone {boneName=dyn_Cloth_volume_l_C01}}, {targetBone {boneName=dyn_Cloth_volume_r_C01}}]
    constraintWeights [1,1]
   }
   n56 = "Constraint - dyn_Cloth_volume_bk02 (PP)"!(ComponentPose=@n55, Alpha=1, ConstraintWeights_0=0.5, ConstraintWeights_1:float=0.5) -> (Pose) {
    alphaBoolBlend {blendOption=Linear}
    boneToModify {boneName=dyn_Cloth_volume_bk02}
    constraintSetup [{targetBone {boneName=dyn_Cloth_volume_l_C02}}, {targetBone {boneName=dyn_Cloth_volume_r_C02}}]
    constraintWeights [1,1]
   }
   Save cached pose 'SkirtCachePose'(Pose=Component To Local(ComponentPose=@n56)) {cacheName=SkirtCachePose}
   n57 = RigidBody!(
    ComponentPose=Local To Component(LocalPose=Use cached pose 'MovingPose'())
    bAlphaBoolEnabled:bool=Get is_WearingCape() -> (is_WearingCape:bool)
    OverrideWorldGravity:Vector="0,0,-980"
    ExternalForce
   ) -> (Pose) {
    alphaInputType=Bool
    bClampLinearTranslationLimitToRefPose=true
    bOverrideWorldGravity=true
    overridePhysicsAsset=&/F_MED_RetroWheels_Physics
    simulationSpace=BaseBoneSpace
    alphaBoolBlend {blendOption=Linear}
    baseBoneRef {boneName=gravity}
    componentAppliedLinearAccClamp=<3000 3000 2700>
    componentLinearAccScale=<0.01 0.4 0.35>
    componentLinearVelScale=<0.08 0.4 0.4>
    overrideWorldGravity {z=-980}
   }
   n58 = Component To Local(ComponentPose=@n57)
   n59 = RigidBody(
    ComponentPose=Local To Component(LocalPose=Use cached pose 'MovingPose'())
    bAlphaBoolEnabled=Get is_WearingCape()
    OverrideWorldGravity="0,0,-980"
   ) {
    alphaInputType=Bool
    bClampLinearTranslationLimitToRefPose=true
    bOverrideWorldGravity=true
    overridePhysicsAsset=&F_MED_RetroWheels_Physics
    simulationSpace=BaseBoneSpace
    alphaBoolBlend {blendOption=Linear}
    baseBoneRef {boneName=gravity}
    componentAppliedLinearAccClamp=<3000 3000 2700>
    componentLinearAccScale=<0.01 0.4 0.35>
    componentLinearVelScale=<0.08 0.4 0.4>
    overrideWorldGravity {z=-980}
   }
   n60 = Component To Local(ComponentPose=@n59)
   n61 = Get MovingAnd NotCrouchNotDivingNotEmote()
   n62 = RigidBody(
    ComponentPose=Local To Component(LocalPose=Use cached pose 'MovingPose'())
    bAlphaBoolEnabled=Get is_WearingCape()
    OverrideWorldGravity="0,0,-980"
   ) {
    alphaInputType=Bool
    bClampLinearTranslationLimitToRefPose=true
    bOverrideWorldGravity=true
    overridePhysicsAsset=&F_MED_RetroWheels_Physics
    simulationSpace=BaseBoneSpace
    alphaBoolBlend {blendOption=Linear}
    baseBoneRef {boneName=gravity}
    componentAppliedLinearAccClamp=<3000 3000 2700>
    componentLinearAccScale=<0.01 0.4 0.35>
    componentLinearVelScale=<0.08 0.4 0.4>
    overrideWorldGravity {z=-980}
   }
   Save cached pose 'SkirtDynPose'(Pose=Blend Poses by bool(
     BlendPose_0=Component To Local(ComponentPose=@n62)
     BlendPose_1=Blend Poses by bool(BlendPose_0=@n60, BlendPose_1=@n58, BlendTime_1=0.5, bActiveValue=@n61) {blendType=Linear}
     BlendTime_1=0.8
     bActiveValue=Get Is Crouhing and jumping and Not Sliding() -> (Is Crouhing and jumping and Not Sliding:bool)
    ) {blendType=Linear}) {cacheName=SkirtDynPose}
   n63 = Anim Dynamics - Bone: dyn_Cloth_volume_l_A!(
    ComponentPose=Local To Component(LocalPose=Use cached pose 'SkirtCachePose'())
    Alpha=1
    GravityOverride:Vector="0,0,500"
    ExternalForce="0,0,0"
   ) -> (Pose) {
    angularDampingOverride=0.99
    angularSpringConstant=800
    bAngularSpring=true
    bOverrideAngularDamping=true
    bUseGravityOverride=true
    gravityScale=0.3
    simulationSpace=BoneRelative
    alphaBoolBlend {blendOption=Linear}
    boundBone {boneName=dyn_Cloth_volume_l_A}
    componentAppliedLinearAccClamp=<5000 5000 5000>
    componentLinearAccScale=<0.45 0.45 0.65>
    componentLinearVelScale=<0.55 0.55 0.7>
    gravityOverride {z=500}
    physicsBodyDefinitions [{
      collisionType=CustomSphere
      sphereCollisionRadius=2
      boundBone {boneName=dyn_Cloth_volume_l_A}
      boxExtents=<30 30 30>
      constraintSetup {angularLimitsMax=<0.1 0.1 20>, angularLimitsMin=<-0.1 -0.1 -2>, angularTarget=<0 0 0>}
      localJointOffset=<-5 0.2 1>
     }]
    relativeSpaceBone {boneName=gravity}
   }
   n64 = Anim Dynamics - Bone: dyn_Cloth_volume_l_A01!(ComponentPose=@n63, Alpha=1, GravityOverride:Vector="0,0,500", ExternalForce="0,0,0") -> (Pose) {
    angularDampingOverride=0.99
    angularSpringConstant=800
    bAngularSpring=true
    bOverrideAngularDamping=true
    bUseGravityOverride=true
    gravityScale=0.3
    simulationSpace=BoneRelative
    alphaBoolBlend {blendOption=Linear}
    boundBone {boneName=dyn_Cloth_volume_l_A01}
    componentAppliedLinearAccClamp=<5000 5000 5000>
    componentLinearAccScale=<0.45 0.45 0.65>
    componentLinearVelScale=<0.55 0.55 0.7>
    gravityOverride {z=500}
    physicsBodyDefinitions [{
      collisionType=CustomSphere
      sphereCollisionRadius=2
      boundBone {boneName=dyn_Cloth_volume_l_A01}
      boxExtents=<20 20 20>
      constraintSetup {angularLimitsMax=<0.1 0.1 20>, angularLimitsMin=<-0.1 -0.1 -2>, angularTarget=<0 0 0>}
      localJointOffset=<-5 0.2 1>
     }]
    relativeSpaceBone {boneName=gravity}
   }
   n65 = Anim Dynamics - Bone: dyn_Cloth_volume_l_B!(ComponentPose=@n64, Alpha=1, GravityOverride:Vector="0,0,500", ExternalForce="0,0,0") -> (Pose) {
    angularDampingOverride=0.99
    angularSpringConstant=800
    bAngularSpring=true
    bOverrideAngularDamping=true
    bUseGravityOverride=true
    gravityScale=0.3
    simulationSpace=BoneRelative
    alphaBoolBlend {blendOption=Linear}
    boundBone {boneName=dyn_Cloth_volume_l_B}
    componentAppliedLinearAccClamp=<5000 5000 5000>
    componentLinearAccScale=<0.4 0.4 0.5>
    componentLinearVelScale=<0.45 0.45 0.7>
    gravityOverride {z=500}
    physicsBodyDefinitions [{
      collisionType=CustomSphere
      sphereCollisionRadius=2
      boundBone {boneName=dyn_Cloth_volume_l_B}
      boxExtents=<50 50 50>
      constraintSetup {angularLimitsMax=<1 1 10>, angularLimitsMin=<-1 -1 0>, angularTarget=<0 0 0>}
      localJointOffset=<-5 -1 -0.1>
     }]
    relativeSpaceBone {boneName=gravity}
   }
   n66 = Anim Dynamics - Bone: dyn_Cloth_volume_l_B01!(ComponentPose=@n65, Alpha=1, GravityOverride:Vector="0,0,500", ExternalForce="0,0,0") -> (Pose) {
    angularDampingOverride=0.99
    angularSpringConstant=800
    bAngularSpring=true
    bOverrideAngularDamping=true
    bUseGravityOverride=true
    gravityScale=0.3
    simulationSpace=BoneRelative
    alphaBoolBlend {blendOption=Linear}
    boundBone {boneName=dyn_Cloth_volume_l_B01}
    componentAppliedLinearAccClamp=<5000 5000 5000>
    componentLinearAccScale=<0.4 0.4 0.5>
    componentLinearVelScale=<0.45 0.45 0.7>
    gravityOverride {z=500}
    physicsBodyDefinitions [{
      collisionType=CustomSphere
      sphereCollisionRadius=2
      boundBone {boneName=dyn_Cloth_volume_l_B01}
      boxExtents=<33 33 33>
      constraintSetup {angularLimitsMax=<1 1 10>, angularLimitsMin=<-1 -1 0>, angularTarget=<0 0 0>}
      localJointOffset=<-5 -1 -0.1>
     }]
    relativeSpaceBone {boneName=gravity}
   }
   n67 = Anim Dynamics - Bone: dyn_Cloth_volume_l_B02!(ComponentPose=@n66, Alpha=1, GravityOverride:Vector="0,0,500", ExternalForce="0,0,0") -> (Pose) {
    angularDampingOverride=0.99
    angularSpringConstant=800
    bAngularSpring=true
    bOverrideAngularDamping=true
    bUseGravityOverride=true
    gravityScale=0.3
    simulationSpace=BoneRelative
    alphaBoolBlend {blendOption=Linear}
    boundBone {boneName=dyn_Cloth_volume_l_B02}
    componentAppliedLinearAccClamp=<5000 5000 5000>
    componentLinearAccScale=<0.4 0.4 0.5>
    componentLinearVelScale=<0.45 0.45 0.7>
    gravityOverride {z=500}
    physicsBodyDefinitions [{
      collisionType=CustomSphere
      sphereCollisionRadius=2
      boundBone {boneName=dyn_Cloth_volume_l_B02}
      boxExtents=<27 27 27>
      constraintSetup {angularLimitsMax=<1 1 10>, angularLimitsMin=<-1 -1 0>, angularTarget=<0 0 0>}
      localJointOffset=<-5 -1 -0.1>
     }]
    relativeSpaceBone {boneName=gravity}
   }
   n68 = Anim Dynamics - Bone: dyn_Cloth_volume_l_C!(ComponentPose=@n67, Alpha=1, GravityOverride:Vector="0,0,500", ExternalForce="0,0,0") -> (Pose) {
    angularDampingOverride=0.99
    angularSpringConstant=800
    bAngularSpring=true
    bOverrideAngularDamping=true
    bUseGravityOverride=true
    gravityScale=0.3
    simulationSpace=BoneRelative
    alphaBoolBlend {blendOption=Linear}
    boundBone {boneName=dyn_Cloth_volume_l_C}
    componentAppliedLinearAccClamp=<5000 5000 5000>
    componentLinearAccScale=<0.5 0.5 0.65>
    componentLinearVelScale=<0.55 0.55 0.7>
    gravityOverride {z=500}
    physicsBodyDefinitions [{
      collisionType=CustomSphere
      sphereCollisionRadius=2
      boundBone {boneName=dyn_Cloth_volume_l_C}
      boxExtents=<60 60 60>
      constraintSetup {angularLimitsMax=<5 5 15>, angularLimitsMin=<-5 -5 -2>, angularTarget=<0 0 0>}
      localJointOffset=<-5 0.2 -0.1>
     }]
    relativeSpaceBone {boneName=gravity}
   }
   n69 = Anim Dynamics - Bone: dyn_Cloth_volume_l_C(ComponentPose=@n68, Alpha=1, GravityOverride="0,0,500", ExternalForce="0,0,0") {
    angularDampingOverride=0.99
    angularSpringConstant=800
    bAngularSpring=true
    bOverrideAngularDamping=true
    bUseGravityOverride=true
    gravityScale=0.3
    simulationSpace=BoneRelative
    alphaBoolBlend {blendOption=Linear}
    boundBone {boneName=dyn_Cloth_volume_l_C}
    componentAppliedLinearAccClamp=<5000 5000 5000>
    componentLinearAccScale=<0.5 0.5 0.65>
    componentLinearVelScale=<0.55 0.55 0.7>
    gravityOverride {z=500}
    physicsBodyDefinitions [{
      collisionType=CustomSphere
      sphereCollisionRadius=2
      boundBone {boneName=dyn_Cloth_volume_l_C}
      boxExtents=<50 50 50>
      constraintSetup {angularLimitsMax=<5 5 15>, angularLimitsMin=<-5 -5 -2>, angularTarget=<0 0 0>}
      localJointOffset=<-5 0.2 -0.1>
     }]
    relativeSpaceBone {boneName=gravity}
   }
   n70 = Anim Dynamics - Bone: dyn_Cloth_volume_l_C02!(ComponentPose=@n69, Alpha=1, GravityOverride:Vector="0,0,500", ExternalForce="0,0,0") -> (Pose) {
    angularDampingOverride=0.99
    angularSpringConstant=800
    bAngularSpring=true
    bOverrideAngularDamping=true
    bUseGravityOverride=true
    gravityScale=0.3
    simulationSpace=BoneRelative
    alphaBoolBlend {blendOption=Linear}
    boundBone {boneName=dyn_Cloth_volume_l_C02}
    componentAppliedLinearAccClamp=<5000 5000 5000>
    componentLinearAccScale=<0.5 0.5 0.65>
    componentLinearVelScale=<0.55 0.55 0.7>
    gravityOverride {z=500}
    physicsBodyDefinitions [{
      collisionType=CustomSphere
      sphereCollisionRadius=2
      boundBone {boneName=dyn_Cloth_volume_l_C02}
      boxExtents=<40 40 40>
      constraintSetup {angularLimitsMax=<5 5 20>, angularLimitsMin=<-5 -5 -2>, angularTarget=<0 0 0>}
      localJointOffset=<-5 0.2 -0.1>
     }]
    relativeSpaceBone {boneName=gravity}
   }
   n71 = Anim Dynamics - Bone: dyn_Cloth_volume_r_A!(ComponentPose=@n70, Alpha=1, GravityOverride:Vector="0,0,-980", ExternalForce="0,0,0") -> (Pose) {
    angularDampingOverride=0.99
    angularSpringConstant=800
    bAngularSpring=true
    bOverrideAngularDamping=true
    bUseGravityOverride=true
    gravityScale=0.3
    simulationSpace=BoneRelative
    alphaBoolBlend {blendOption=Linear}
    boundBone {boneName=dyn_Cloth_volume_r_A}
    componentAppliedLinearAccClamp=<5000 5000 5000>
    componentLinearAccScale=<0.45 0.45 0.65>
    componentLinearVelScale=<0.55 0.55 0.7>
    gravityOverride {z=-980}
    physicsBodyDefinitions [{
      collisionType=CustomSphere
      sphereCollisionRadius=2
      boundBone {boneName=dyn_Cloth_volume_r_A}
      boxExtents=<30 30 30>
      constraintSetup {angularLimitsMax=<1 1 20>, angularLimitsMin=<-1 -1 -2>, angularTarget=<0 0 0>}
      localJointOffset=<-5 0.2 1>
     }]
    relativeSpaceBone {boneName=gravity}
   }
   n72 = Anim Dynamics - Bone: dyn_Cloth_volume_r_A01!(ComponentPose=@n71, Alpha=1, GravityOverride:Vector="0,0,-980", ExternalForce="0,0,0") -> (Pose) {
    angularDampingOverride=0.99
    angularSpringConstant=800
    bAngularSpring=true
    bOverrideAngularDamping=true
    bUseGravityOverride=true
    gravityScale=0.3
    simulationSpace=BoneRelative
    alphaBoolBlend {blendOption=Linear}
    boundBone {boneName=dyn_Cloth_volume_r_A01}
    componentAppliedLinearAccClamp=<5000 5000 5000>
    componentLinearAccScale=<0.45 0.45 0.65>
    componentLinearVelScale=<0.55 0.55 0.7>
    gravityOverride {z=-980}
    physicsBodyDefinitions [{
      collisionType=CustomSphere
      sphereCollisionRadius=2
      boundBone {boneName=dyn_Cloth_volume_r_A01}
      boxExtents=<20 20 20>
      constraintSetup {angularLimitsMax=<1 1 20>, angularLimitsMin=<-1 -1 -2>, angularTarget=<0 0 0>}
      localJointOffset=<-5 0.2 1>
     }]
    relativeSpaceBone {boneName=gravity}
   }
   n73 = Anim Dynamics - Bone: dyn_Cloth_volume_r_B!(ComponentPose=@n72, Alpha=1, GravityOverride:Vector="0,0,-980", ExternalForce="0,0,0") -> (Pose) {
    angularDampingOverride=0.99
    angularSpringConstant=800
    bAngularSpring=true
    bOverrideAngularDamping=true
    bUseGravityOverride=true
    gravityScale=0.3
    simulationSpace=BoneRelative
    alphaBoolBlend {blendOption=Linear}
    boundBone {boneName=dyn_Cloth_volume_r_B}
    componentAppliedLinearAccClamp=<5000 5000 5000>
    componentLinearAccScale=<0.4 0.4 0.5>
    componentLinearVelScale=<0.45 0.45 0.7>
    gravityOverride {z=-980}
    physicsBodyDefinitions [{
      collisionType=CustomSphere
      sphereCollisionRadius=2
      boundBone {boneName=dyn_Cloth_volume_r_B}
      boxExtents=<50 50 50>
      constraintSetup {angularLimitsMax=<1 1 10>, angularLimitsMin=<-1 -1 0>, angularTarget=<0 0 0>}
      localJointOffset=<-5 -1 -0.1>
     }]
    relativeSpaceBone {boneName=gravity}
   }
   n74 = Anim Dynamics - Bone: dyn_Cloth_volume_r_B01!(ComponentPose=@n73, Alpha=1, GravityOverride:Vector="0,0,-980", ExternalForce="0,0,0") -> (Pose) {
    angularDampingOverride=0.99
    angularSpringConstant=800
    bAngularSpring=true
    bOverrideAngularDamping=true
    bUseGravityOverride=true
    gravityScale=0.3
    simulationSpace=BoneRelative
    alphaBoolBlend {blendOption=Linear}
    boundBone {boneName=dyn_Cloth_volume_r_B01}
    componentAppliedLinearAccClamp=<5000 5000 5000>
    componentLinearAccScale=<0.4 0.4 0.5>
    componentLinearVelScale=<0.45 0.45 0.7>
    gravityOverride {z=-980}
    physicsBodyDefinitions [{
      collisionType=CustomSphere
      sphereCollisionRadius=2
      boundBone {boneName=dyn_Cloth_volume_r_B01}
      boxExtents=<33 33 33>
      constraintSetup {angularLimitsMax=<1 1 10>, angularLimitsMin=<-1 -1 0>, angularTarget=<0 0 0>}
      localJointOffset=<-5 -1 -0.1>
     }]
    relativeSpaceBone {boneName=gravity}
   }
   n75 = Anim Dynamics - Bone: dyn_Cloth_volume_r_B02!(ComponentPose=@n74, Alpha=1, GravityOverride:Vector="0,0,-980", ExternalForce="0,0,0") -> (Pose) {
    angularDampingOverride=0.99
    angularSpringConstant=800
    bAngularSpring=true
    bOverrideAngularDamping=true
    bUseGravityOverride=true
    gravityScale=0.3
    simulationSpace=BoneRelative
    alphaBoolBlend {blendOption=Linear}
    boundBone {boneName=dyn_Cloth_volume_r_B02}
    componentAppliedLinearAccClamp=<5000 5000 5000>
    componentLinearAccScale=<0.4 0.4 0.5>
    componentLinearVelScale=<0.45 0.45 0.7>
    gravityOverride {z=-980}
    physicsBodyDefinitions [{
      collisionType=CustomSphere
      sphereCollisionRadius=2
      boundBone {boneName=dyn_Cloth_volume_r_B02}
      boxExtents=<27 27 27>
      constraintSetup {angularLimitsMax=<1 1 10>, angularLimitsMin=<-1 -1 0>, angularTarget=<0 0 0>}
      localJointOffset=<-5 -1 -0.1>
     }]
    relativeSpaceBone {boneName=gravity}
   }
   n76 = Anim Dynamics - Bone: dyn_Cloth_volume_r_C!(ComponentPose=@n75, Alpha=1, GravityOverride:Vector="0,0,-980", ExternalForce="0,0,0") -> (Pose) {
    angularDampingOverride=0.99
    angularSpringConstant=800
    bAngularSpring=true
    bOverrideAngularDamping=true
    bUseGravityOverride=true
    gravityScale=0.3
    simulationSpace=BoneRelative
    alphaBoolBlend {blendOption=Linear}
    boundBone {boneName=dyn_Cloth_volume_r_C}
    componentAppliedLinearAccClamp=<5000 5000 5000>
    componentLinearAccScale=<0.5 0.5 0.65>
    componentLinearVelScale=<0.55 0.55 0.7>
    gravityOverride {z=-980}
    physicsBodyDefinitions [{
      collisionType=CustomSphere
      sphereCollisionRadius=2
      boundBone {boneName=dyn_Cloth_volume_r_C}
      boxExtents=<45 45 45>
      constraintSetup {angularLimitsMax=<5 5 15>, angularLimitsMin=<-5 -5 -2>, angularTarget=<0 0 0>}
      localJointOffset=<-5 0.2 -0.1>
     }]
    relativeSpaceBone {boneName=gravity}
   }
   n77 = Anim Dynamics - Bone: dyn_Cloth_volume_r_C01!(ComponentPose=@n76, Alpha=1, GravityOverride:Vector="0,0,-980", ExternalForce="0,0,0") -> (Pose) {
    angularDampingOverride=0.99
    angularSpringConstant=800
    bAngularSpring=true
    bOverrideAngularDamping=true
    bUseGravityOverride=true
    gravityScale=0.3
    simulationSpace=BoneRelative
    alphaBoolBlend {blendOption=Linear}
    boundBone {boneName=dyn_Cloth_volume_r_C01}
    componentAppliedLinearAccClamp=<5000 5000 5000>
    componentLinearAccScale=<0.5 0.5 0.65>
    componentLinearVelScale=<0.55 0.55 0.7>
    gravityOverride {z=-980}
    physicsBodyDefinitions [{
      collisionType=CustomSphere
      sphereCollisionRadius=2
      boundBone {boneName=dyn_Cloth_volume_r_C01}
      boxExtents=<40 40 40>
      constraintSetup {angularLimitsMax=<5 5 15>, angularLimitsMin=<-5 -5 -2>, angularTarget=<0 0 0>}
      localJointOffset=<-5 0.2 -0.1>
     }]
    relativeSpaceBone {boneName=gravity}
   }
   n78 = Anim Dynamics - Bone: dyn_Cloth_volume_r_C02!(ComponentPose=@n77, Alpha=1, GravityOverride:Vector="0,0,-980", ExternalForce="0,0,0") -> (Pose) {
    angularDampingOverride=0.99
    angularSpringConstant=800
    bAngularSpring=true
    bOverrideAngularDamping=true
    bUseGravityOverride=true
    gravityScale=0.3
    simulationSpace=BoneRelative
    alphaBoolBlend {blendOption=Linear}
    boundBone {boneName=dyn_Cloth_volume_r_C02}
    componentAppliedLinearAccClamp=<5000 5000 5000>
    componentLinearAccScale=<0.5 0.5 0.65>
    componentLinearVelScale=<0.55 0.55 0.7>
    gravityOverride {z=-980}
    physicsBodyDefinitions [{
      collisionType=CustomSphere
      sphereCollisionRadius=2
      boundBone {boneName=dyn_Cloth_volume_r_C02}
      boxExtents=<33 33 33>
      constraintSetup {angularLimitsMax=<5 5 20>, angularLimitsMin=<-5 -5 -2>, angularTarget=<0 0 0>}
      localJointOffset=<-5 0.2 -0.1>
     }]
    relativeSpaceBone {boneName=gravity}
   }
   n79 = Anim Dynamics - Bone: dyn_Cloth_volume_bk!(ComponentPose=@n78, Alpha=1, GravityOverride:Vector="0,0,-980", ExternalForce="0,0,0") -> (Pose) {
    angularDampingOverride=0.99
    angularSpringConstant=800
    bAngularSpring=true
    bOverrideAngularDamping=true
    bUseGravityOverride=true
    gravityScale=0.3
    simulationSpace=BoneRelative
    alphaBoolBlend {blendOption=Linear}
    boundBone {boneName=dyn_Cloth_volume_bk}
    componentAppliedLinearAccClamp=<5000 5000 5000>
    componentLinearAccScale=<0.5 0.5 0.65>
    componentLinearVelScale=<0.55 0.55 0.7>
    gravityOverride {z=-980}
    physicsBodyDefinitions [{
      collisionType=CustomSphere
      sphereCollisionRadius=2
      boundBone {boneName=dyn_Cloth_volume_bk}
      boxExtents=<45 45 45>
      constraintSetup {angularLimitsMax=<5 5 20>, angularLimitsMin=<-5 -5 -2>, angularTarget=<0 0 0>}
      localJointOffset=<-5 1 -0.1>
     }]
    relativeSpaceBone {boneName=gravity}
   }
   n80 = Anim Dynamics - Bone: dyn_Cloth_volume_bk01!(ComponentPose=@n79, Alpha=1, GravityOverride:Vector="0,0,-980", ExternalForce="0,0,0") -> (Pose) {
    angularDampingOverride=0.99
    angularSpringConstant=800
    bAngularSpring=true
    bOverrideAngularDamping=true
    bUseGravityOverride=true
    gravityScale=0.3
    simulationSpace=BoneRelative
    alphaBoolBlend {blendOption=Linear}
    boundBone {boneName=dyn_Cloth_volume_bk01}
    componentAppliedLinearAccClamp=<5000 5000 5000>
    componentLinearAccScale=<0.5 0.5 0.65>
    componentLinearVelScale=<0.55 0.55 0.7>
    gravityOverride {z=-980}
    physicsBodyDefinitions [{
      collisionType=CustomSphere
      sphereCollisionRadius=2
      boundBone {boneName=dyn_Cloth_volume_bk01}
      boxExtents=<40 40 40>
      constraintSetup {angularLimitsMax=<5 5 20>, angularLimitsMin=<-5 -5 -2>, angularTarget=<0 0 0>}
      localJointOffset=<-5 1 -0.1>
     }]
    relativeSpaceBone {boneName=gravity}
   }
   n81 = Anim Dynamics - Bone: dyn_Cloth_volume_bk02!(ComponentPose=@n80, Alpha=1, GravityOverride:Vector="0,0,-980", ExternalForce="0,0,0") -> (Pose) {
    angularDampingOverride=0.99
    angularSpringConstant=800
    bAngularSpring=true
    bOverrideAngularDamping=true
    bUseGravityOverride=true
    gravityScale=0.3
    simulationSpace=BoneRelative
    alphaBoolBlend {blendOption=Linear}
    boundBone {boneName=dyn_Cloth_volume_bk02}
    componentAppliedLinearAccClamp=<5000 5000 5000>
    componentLinearAccScale=<0.5 0.5 0.65>
    componentLinearVelScale=<0.55 0.55 0.7>
    gravityOverride {z=-980}
    physicsBodyDefinitions [{
      collisionType=CustomSphere
      sphereCollisionRadius=2
      boundBone {boneName=dyn_Cloth_volume_bk02}
      boxExtents=<33 33 33>
      constraintSetup {angularLimitsMax=<5 5 20>, angularLimitsMin=<-5 -5 -2>, angularTarget=<0 0 0>}
      localJointOffset=<-5 1 -0.1>
     }]
    relativeSpaceBone {boneName=gravity}
   }
   n82 = "Constraint - dyn_Cloth_volume_l_C (PPP)"!(ComponentPose=@n81, Alpha=1, ConstraintWeights_0=0.9, ConstraintWeights_1:float=0, ConstraintWeights_2:float=0.1) -> (Pose) {
    alphaBoolBlend {blendOption=Linear}
    boneToModify {boneName=dyn_Cloth_volume_l_C}
    constraintSetup [{targetBone {boneName=dyn_Cloth_volume_l_C}}, {targetBone {boneName=dyn_Cloth_volume_r_C}}, {targetBone {boneName=dyn_Cloth_volume_bk}}]
    constraintWeights [1,1,1]
   }
   n83 = "Constraint - dyn_Cloth_volume_l_C01 (PPP)"!(ComponentPose=@n82, Alpha=1, ConstraintWeights_0=0.9, ConstraintWeights_1:float=0, ConstraintWeights_2:float=0.1) -> (Pose) {
    alphaBoolBlend {blendOption=Linear}
    boneToModify {boneName=dyn_Cloth_volume_l_C01}
    constraintSetup [
     {targetBone {boneName=dyn_Cloth_volume_l_C01}}
     {targetBone {boneName=dyn_Cloth_volume_r_C01}}
     {targetBone {boneName=dyn_Cloth_volume_bk01}}
    ]
    constraintWeights [1,1,1]
   }
   n84 = "Constraint - dyn_Cloth_volume_l_C02 (PPP)"!(ComponentPose=@n83, Alpha=1, ConstraintWeights_0=0.9, ConstraintWeights_1:float=0, ConstraintWeights_2:float=0.1) -> (Pose) {
    alphaBoolBlend {blendOption=Linear}
    boneToModify {boneName=dyn_Cloth_volume_l_C02}
    constraintSetup [
     {targetBone {boneName=dyn_Cloth_volume_l_C02}}
     {targetBone {boneName=dyn_Cloth_volume_r_C02}}
     {targetBone {boneName=dyn_Cloth_volume_bk02}}
    ]
    constraintWeights [1,1,1]
   }
   n85 = "Constraint - dyn_Cloth_volume_r_C (PPP)"!(ComponentPose=@n84, Alpha=1, ConstraintWeights_0=0, ConstraintWeights_1:float=0.9, ConstraintWeights_2:float=0.1) -> (Pose) {
    alphaBoolBlend {blendOption=Linear}
    boneToModify {boneName=dyn_Cloth_volume_r_C}
    constraintSetup [{targetBone {boneName=dyn_Cloth_volume_l_C}}, {targetBone {boneName=dyn_Cloth_volume_r_C}}, {targetBone {boneName=dyn_Cloth_volume_bk}}]
    constraintWeights [1,1,1]
   }
   n86 = "Constraint - dyn_Cloth_volume_r_C01 (PPP)"!(ComponentPose=@n85, Alpha=1, ConstraintWeights_0=0, ConstraintWeights_1:float=0.9, ConstraintWeights_2:float=0.1) -> (Pose) {
    alphaBoolBlend {blendOption=Linear}
    boneToModify {boneName=dyn_Cloth_volume_r_C01}
    constraintSetup [
     {targetBone {boneName=dyn_Cloth_volume_l_C01}}
     {targetBone {boneName=dyn_Cloth_volume_r_C01}}
     {targetBone {boneName=dyn_Cloth_volume_bk01}}
    ]
    constraintWeights [1,1,1]
   }
   n87 = "Constraint - dyn_Cloth_volume_r_C02 (PPP)"!(ComponentPose=@n86, Alpha=1, ConstraintWeights_0=0, ConstraintWeights_1:float=0.9, ConstraintWeights_2:float=0.1) -> (Pose) {
    alphaBoolBlend {blendOption=Linear}
    boneToModify {boneName=dyn_Cloth_volume_r_C02}
    constraintSetup [
     {targetBone {boneName=dyn_Cloth_volume_l_C02}}
     {targetBone {boneName=dyn_Cloth_volume_r_C02}}
     {targetBone {boneName=dyn_Cloth_volume_bk02}}
    ]
    constraintWeights [1,1,1]
   }
   n88 = "Constraint - dyn_Cloth_volume_l_B (PPP)"!(ComponentPose=@n87, Alpha=1, ConstraintWeights_0=0.8, ConstraintWeights_1:float=0.1, ConstraintWeights_2:float=0.1) -> (Pose) {
    alphaBoolBlend {blendOption=Linear}
    boneToModify {boneName=dyn_Cloth_volume_l_B}
    constraintSetup [{targetBone {boneName=dyn_Cloth_volume_l_B}}, {targetBone {boneName=dyn_Cloth_volume_l_Up02}}, {targetBone {boneName=dyn_Cloth_volume_l_C}}]
    constraintWeights [1,1,1]
   }
   n89 = "Constraint - dyn_Cloth_volume_r_B (PPP)"!(ComponentPose=@n88, Alpha=1, ConstraintWeights_0=0.8, ConstraintWeights_1:float=0.1, ConstraintWeights_2:float=0.1) -> (Pose) {
    alphaBoolBlend {blendOption=Linear}
    boneToModify {boneName=dyn_Cloth_volume_r_B}
    constraintSetup [{targetBone {boneName=dyn_Cloth_volume_r_B}}, {targetBone {boneName=dyn_Cloth_volume_r_Up02}}, {targetBone {boneName=dyn_Cloth_volume_r_C}}]
    constraintWeights [1,1,1]
   }
   Save cached pose 'MovingPose'(Pose=Component To Local(ComponentPose=@n89)) {cacheName=MovingPose}
   n90 = "Constraint - dyn_Cloth_volume_r_C02 (RR)"!(
    ComponentPose=Local To Component(LocalPose=Use cached pose 'SkirtCachePose'())
    Alpha=1
    ConstraintWeights_0=0.4
    ConstraintWeights_1:float=0.6
   ) -> (Pose) {
    alphaBoolBlend {blendOption=Linear}
    boneToModify {boneName=dyn_Cloth_volume_r_C02}
    constraintSetup [{transformType=Rotation, targetBone {boneName=calf_r}}, {transformType=Rotation, targetBone {boneName=dyn_Cloth_volume_r_C01}}]
    constraintWeights [1,1]
   }
   n91 = "Constraint - dyn_Cloth_volume_l_C02 (RR)"!(ComponentPose=@n90, Alpha=1, ConstraintWeights_0=0.4, ConstraintWeights_1:float=0.6) -> (Pose) {
    alphaBoolBlend {blendOption=Linear}
    boneToModify {boneName=dyn_Cloth_volume_l_C02}
    constraintSetup [{transformType=Rotation, targetBone {boneName=calf_l}}, {transformType=Rotation, targetBone {boneName=dyn_Cloth_volume_l_C01}}]
    constraintWeights [1,1]
   }
   n92 = "Constraint - dyn_Cloth_volume_r_B (PP)"(ComponentPose=@n91, Alpha=1, ConstraintWeights_0=0.4, ConstraintWeights_1=0.6) {
    alphaBoolBlend {blendOption=Linear}
    boneToModify {boneName=dyn_Cloth_volume_r_B}
    constraintSetup [{targetBone {boneName=dyn_Cloth_volume_r_A}}, {targetBone {boneName=dyn_Cloth_volume_r_B}}]
    constraintWeights [1,1]
   }
   n93 = "Constraint - dyn_Cloth_volume_l_B (PP)"(ComponentPose=@n92, Alpha=1, ConstraintWeights_0=0.4, ConstraintWeights_1=0.6) {
    alphaBoolBlend {blendOption=Linear}
    boneToModify {boneName=dyn_Cloth_volume_l_B}
    constraintSetup [{targetBone {boneName=dyn_Cloth_volume_l_A}}, {targetBone {boneName=dyn_Cloth_volume_l_B}}]
    constraintWeights [1,1]
   }
   n94 = Component To Local(ComponentPose=@n93)
   n95 = Use cached pose 'SkirtCachePose'()
   n96 = Get Not DiveFloatingAndDiveMode() -> (Not DiveFloatingAndDiveMode:bool)
   n97 = "Constraint - dyn_Cloth_volume_bk (PP)"(
    ComponentPose=Local To Component(LocalPose=Blend Poses by bool(BlendPose_0=@n95, BlendPose_1=@n94, BlendTime_0=0.5, bActiveValue=@n96) {blendType=Linear})
    Alpha=1
    ConstraintWeights_0=0.5
    ConstraintWeights_1=0.5
   ) {
    alphaBoolBlend {blendOption=Linear}
    boneToModify {boneName=dyn_Cloth_volume_bk}
    constraintSetup [{targetBone {boneName=dyn_Cloth_volume_l_C}}, {targetBone {boneName=dyn_Cloth_volume_r_C}}]
    constraintWeights [1,1]
   }
   n98 = "Constraint - dyn_Cloth_volume_bk01 (PP)"(ComponentPose=@n97, Alpha=1, ConstraintWeights_0=0.5, ConstraintWeights_1=0.5) {
    alphaBoolBlend {blendOption=Linear}
    boneToModify {boneName=dyn_Cloth_volume_bk01}
    constraintSetup [{targetBone {boneName=dyn_Cloth_volume_l_C01}}, {targetBone {boneName=dyn_Cloth_volume_r_C01}}]
    constraintWeights [1,1]
   }
   n99 = "Constraint - dyn_Cloth_volume_bk02 (PP)"(ComponentPose=@n98, Alpha=1, ConstraintWeights_0=0.5, ConstraintWeights_1=0.5) {
    alphaBoolBlend {blendOption=Linear}
    boneToModify {boneName=dyn_Cloth_volume_bk02}
    constraintSetup [{targetBone {boneName=dyn_Cloth_volume_l_C02}}, {targetBone {boneName=dyn_Cloth_volume_r_C02}}]
    constraintWeights [1,1]
   }
   Save cached pose 'DBNOPose'(Pose=Component To Local(ComponentPose=@n99)) {cacheName=DBNOPose}
   n100 = "Constraint - attach_cape (PPPPP)"!(
    ComponentPose=Local To Component(LocalPose=Copy Pose From Mesh() {bUseAttachedParent=true})
    Alpha=1
    ConstraintWeights_0=0.8787
    ConstraintWeights_1:float=0.0589
    ConstraintWeights_2:float=0.0389
    ConstraintWeights_3:float=0.0234
    ConstraintWeights_4:float=0.0002
   ) -> (Pose) {
    alphaBoolBlend {blendOption=Linear}
    boneToModify {boneName=attach_cape}
    constraintSetup [
     {targetBone {boneName=spine_05}}
     {targetBone {boneName=spine_04}}
     {targetBone {boneName=neck_02}}
     {targetBone {boneName=neck_01}}
     {targetBone {boneName=clavicle_r}}
    ]
    constraintWeights [1,1,1,1,1]
   }
   n101 = "Constraint - attach_backpack (PPP)"!(ComponentPose=@n100, Alpha=1, ConstraintWeights_0=0.5563, ConstraintWeights_1:float=0.3613, ConstraintWeights_2:float=0.0824) -> (Pose) {
    alphaBoolBlend {blendOption=Linear}
    boneToModify {boneName=attach_backpack}
    constraintSetup [{targetBone {boneName=spine_05}}, {targetBone {boneName=spine_04}}, {targetBone {boneName=spine_03}}]
    constraintWeights [1,1,1]
   }
   n102 = "Constraint - gravity (RR)"!(ComponentPose=@n101, Alpha=1, ConstraintWeights_0=1, ConstraintWeights_1:float=0) -> (Pose) {
    alphaBoolBlend {blendOption=Linear}
    boneToModify {boneName=gravity}
    constraintSetup [{transformType=Rotation, targetBone {boneName=root}}, {transformType=Rotation, targetBone {boneName=spine_01}}]
    constraintWeights [1,1]
   }
   n103 = "Constraint - gravity (TT)"!(ComponentPose=@n102, Alpha=1, ConstraintWeights_0=0.3, ConstraintWeights_1:float=0.7) -> (Pose) {
    alphaBoolBlend {blendOption=Linear}
    boneToModify {boneName=gravity}
    constraintSetup [{transformType=Translation, targetBone {boneName=root}}, {transformType=Translation, targetBone {boneName=spine_01}}]
    constraintWeights [1,1]
   }
   n104 = "Transform (Modify) Bone - Bone: gravity"(ComponentPose=@n103, Alpha=1, Translation="0,0,0", Rotation="0,0,0", Scale="1,1,1") {
    rotationMode=BMM_Replace
    scaleMode=BMM_Additive
    scaleSpace=BCS_BoneSpace
    translationMode=BMM_Additive
    translationSpace=BCS_BoneSpace
    alphaBoolBlend {blendOption=Linear}
    boneToModify {boneName=gravity}
   }
   n105 = "Transform (Modify) Bone - Bone: upperarmFix_l"(ComponentPose=@n104, Alpha=1, Translation="0,0,0", Rotation="0,0,0", Scale="0.3,0.3,0.3") {
    rotationMode=BMM_Additive
    rotationSpace=BCS_BoneSpace
    scaleMode=BMM_Additive
    scaleSpace=BCS_BoneSpace
    translationMode=BMM_Additive
    translationSpace=BCS_BoneSpace
    alphaBoolBlend {blendOption=Linear}
    boneToModify {boneName=upperarmFix_l}
    scale=<0.3 0.3 0.3>
   }
   n106 = "Transform (Modify) Bone - Bone: upperarmFix_r"(ComponentPose=@n105, Alpha=1, Translation="0,0,0", Rotation="0,0,0", Scale="0.3,0.3,0.3") {
    rotationMode=BMM_Additive
    rotationSpace=BCS_BoneSpace
    scaleMode=BMM_Additive
    scaleSpace=BCS_BoneSpace
    translationMode=BMM_Additive
    translationSpace=BCS_BoneSpace
    alphaBoolBlend {blendOption=Linear}
    boneToModify {boneName=upperarmFix_r}
    scale=<0.3 0.3 0.3>
   }
   Save cached pose 'bodyPose'(Pose=Component To Local(ComponentPose=@n106)) {cacheName=bodyPose}
   n107 = Get is sliding() -> (is sliding:double)
   n108 = "Constraint - dyn_Cloth_volume_l_C01 (PP)"(ComponentPose=Local To Component(LocalPose=Use cached pose 'slidingPose'()), Alpha=@n107, ConstraintWeights_0=0.5, ConstraintWeights_1=0.5) {
    alphaBoolBlend {blendOption=Linear}
    boneToModify {boneName=dyn_Cloth_volume_l_C01}
    constraintSetup [{targetBone {boneName=dyn_Cloth_volume_l_C}}, {targetBone {boneName=spine_01}}]
    constraintWeights [1,1]
   }
   n109 = "Constraint - dyn_Cloth_volume_r_C01 (PP)"(ComponentPose=@n108, Alpha=@n107, ConstraintWeights_0=0.5, ConstraintWeights_1=0.5) {
    alphaBoolBlend {blendOption=Linear}
    boneToModify {boneName=dyn_Cloth_volume_r_C01}
    constraintSetup [{targetBone {boneName=dyn_Cloth_volume_r_C}}, {targetBone {boneName=spine_01}}]
    constraintWeights [1,1]
   }
   n110 = "Transform (Modify) Bone - Bone: dyn_Cloth_volume_l_C"(ComponentPose=@n109, Alpha=@n107, Translation="5,0,0", Rotation="20,90,0", Scale="1,1,1") {
    rotationMode=BMM_Additive
    rotationSpace=BCS_BoneSpace
    scaleMode=BMM_Additive
    scaleSpace=BCS_BoneSpace
    translationMode=BMM_Additive
    translationSpace=BCS_BoneSpace
    alphaBoolBlend {blendOption=Linear}
    boneToModify {boneName=dyn_Cloth_volume_l_C}
    rotation {pitch=20, yaw=90}
    translation {x=5}
   }
   n111 = "Transform (Modify) Bone - Bone: dyn_Cloth_volume_r_C"(ComponentPose=@n110, Alpha=@n107, Translation="5,0,0", Rotation="-20,90,0", Scale="1,1,1") {
    rotationMode=BMM_Additive
    rotationSpace=BCS_BoneSpace
    scaleMode=BMM_Additive
    scaleSpace=BCS_BoneSpace
    translationMode=BMM_Additive
    translationSpace=BCS_BoneSpace
    alphaBoolBlend {blendOption=Linear}
    boneToModify {boneName=dyn_Cloth_volume_r_C}
    rotation {pitch=-20, yaw=90}
    translation {x=5}
   }
   n112 = "Transform (Modify) Bone - Bone: dyn_Cloth_volume_l_B"(ComponentPose=@n111, Alpha=@n107, Translation="0,0,0", Rotation="0,10,0", Scale="1,1,1") {
    rotationMode=BMM_Additive
    rotationSpace=BCS_BoneSpace
    scaleMode=BMM_Additive
    scaleSpace=BCS_BoneSpace
    translationMode=BMM_Additive
    translationSpace=BCS_BoneSpace
    alphaBoolBlend {blendOption=Linear}
    boneToModify {boneName=dyn_Cloth_volume_l_B}
    rotation {yaw=10}
   }
   n113 = "Transform (Modify) Bone - Bone: dyn_Cloth_volume_r_B"(ComponentPose=@n112, Alpha=@n107, Translation="0,0,0", Rotation="0,10,0", Scale="1,1,1") {
    rotationMode=BMM_Additive
    rotationSpace=BCS_BoneSpace
    scaleMode=BMM_Additive
    scaleSpace=BCS_BoneSpace
    translationMode=BMM_Additive
    translationSpace=BCS_BoneSpace
    alphaBoolBlend {blendOption=Linear}
    boneToModify {boneName=dyn_Cloth_volume_r_B}
    rotation {yaw=10}
   }
   n114 = "Transform (Modify) Bone - Bone: dyn_Cloth_volume_r_B"(ComponentPose=@n113, Alpha=1, Translation="10,-2,4", Rotation="0,20,0", Scale="1,1,1") {
    rotationMode=BMM_Additive
    rotationSpace=BCS_BoneSpace
    scaleMode=BMM_Additive
    scaleSpace=BCS_BoneSpace
    translationMode=BMM_Additive
    translationSpace=BCS_BoneSpace
    alphaBoolBlend {blendOption=Linear}
    boneToModify {boneName=dyn_Cloth_volume_r_B}
    rotation {yaw=20}
    translation=<10 -2 4>
   }
   n115 = "Transform (Modify) Bone - Bone: dyn_Cloth_volume_r_C"(ComponentPose=@n114, Alpha=1, Translation="5,0,0", Rotation="0,30,0", Scale="1,1,1") {
    rotationMode=BMM_Additive
    rotationSpace=BCS_BoneSpace
    scaleMode=BMM_Additive
    scaleSpace=BCS_BoneSpace
    translationMode=BMM_Additive
    translationSpace=BCS_BoneSpace
    alphaBoolBlend {blendOption=Linear}
    boneToModify {boneName=dyn_Cloth_volume_r_C}
    rotation {yaw=30}
    translation {x=5}
   }
   n116 = "Constraint - dyn_Cloth_volume_r_Up02 (PP)"!(ComponentPose=@n115, Alpha=1, ConstraintWeights_0=0.5, ConstraintWeights_1:float=0.5) -> (Pose) {
    alphaBoolBlend {blendOption=Linear}
    boneToModify {boneName=dyn_Cloth_volume_r_Up02}
    constraintSetup [{targetBone {boneName=thigh_twist_01_r}}, {targetBone {boneName=dyn_Cloth_volume_r_Up02}}]
    constraintWeights [1,1]
   }
   n117 = "Transform (Modify) Bone - Bone: dyn_Cloth_volume_r_Up02"(ComponentPose=@n116, Alpha=1, Translation="0,-2,0", Rotation="-10,10,0", Scale="1,1,1") {
    rotationMode=BMM_Additive
    rotationSpace=BCS_BoneSpace
    scaleMode=BMM_Additive
    scaleSpace=BCS_BoneSpace
    translationMode=BMM_Additive
    translationSpace=BCS_BoneSpace
    alphaBoolBlend {blendOption=Linear}
    boneToModify {boneName=dyn_Cloth_volume_r_Up02}
    rotation {pitch=-10, yaw=10}
    translation {y=-2}
   }
   n118 = "Constraint - dyn_Cloth_volume_l_Up02 (PP)"!(ComponentPose=@n117, Alpha=1, ConstraintWeights_0=0.5, ConstraintWeights_1:float=0.5) -> (Pose) {
    alphaBoolBlend {blendOption=Linear}
    boneToModify {boneName=dyn_Cloth_volume_l_Up02}
    constraintSetup [{targetBone {boneName=thigh_twist_01_l}}, {targetBone {boneName=dyn_Cloth_volume_l_Up02}}]
    constraintWeights [1,1]
   }
   n119 = "Transform (Modify) Bone - Bone: dyn_Cloth_volume_l_Up02"(ComponentPose=@n118, Alpha=1, Translation="0,-2,0", Rotation="-10,40,0", Scale="1,1,1") {
    rotationMode=BMM_Additive
    rotationSpace=BCS_BoneSpace
    scaleMode=BMM_Additive
    scaleSpace=BCS_BoneSpace
    translationMode=BMM_Additive
    translationSpace=BCS_BoneSpace
    alphaBoolBlend {blendOption=Linear}
    boneToModify {boneName=dyn_Cloth_volume_l_Up02}
    rotation {pitch=-10, yaw=40}
    translation {y=-2}
   }
   n120 = Component To Local(ComponentPose=@n119)
   n121 = Use cached pose 'slidingPose'()
   n122 = "Constraint - dyn_Cloth_volume_bk (PP)"(
    ComponentPose=Local To Component(LocalPose=Blend Poses by bool(BlendPose_0=@n120, BlendPose_1=@n121, BlendTime_0=0.5, BlendTime_1=0.5) {blendType=Linear})
    Alpha=1
    ConstraintWeights_0=0.5
    ConstraintWeights_1=0.5
   ) {
    alphaBoolBlend {blendOption=Linear}
    boneToModify {boneName=dyn_Cloth_volume_bk}
    constraintSetup [{targetBone {boneName=dyn_Cloth_volume_l_C}}, {targetBone {boneName=dyn_Cloth_volume_r_C}}]
    constraintWeights [1,1]
   }
   n123 = "Constraint - dyn_Cloth_volume_bk01 (PP)"(ComponentPose=@n122, Alpha=1, ConstraintWeights_0=0.5, ConstraintWeights_1=0.5) {
    alphaBoolBlend {blendOption=Linear}
    boneToModify {boneName=dyn_Cloth_volume_bk01}
    constraintSetup [{targetBone {boneName=dyn_Cloth_volume_l_C01}}, {targetBone {boneName=dyn_Cloth_volume_r_C01}}]
    constraintWeights [1,1]
   }
   n124 = "Constraint - dyn_Cloth_volume_bk02 (PP)"(ComponentPose=@n123, Alpha=1, ConstraintWeights_0=0.5, ConstraintWeights_1=0.5) {
    alphaBoolBlend {blendOption=Linear}
    boneToModify {boneName=dyn_Cloth_volume_bk02}
    constraintSetup [{targetBone {boneName=dyn_Cloth_volume_l_C02}}, {targetBone {boneName=dyn_Cloth_volume_r_C02}}]
    constraintWeights [1,1]
   }
   Save cached pose 'CrouchingPose'(Pose=Component To Local(ComponentPose=@n124)) {cacheName=CrouchingPose}
  ]}
 EventGraph {nodes [
   Event BlueprintUpdateAnimation()
   > n1 = Sequence!(execute) -> (then_0, then_1, then_2:Exec, then_3:Exec)
   > n2 = Gravity Override State Function DEPRECATED(CharacterPartAnimInstance=Self-Reference(), GameTimeInSeconds=0, GravityOverrideParams=Get Gravity Override Params())
   > Set GravityOverride(GravityOverride=@n2)
   n3 = Branch(execute=@n1.then_1, Condition=OR Boolean(A=Get bIsBaseballSliding(), B=Get bIsKneeSliding()))
   > Set DynamicsAlpha(DynamicsAlpha:double=0.25) -> (Output_Get:double)
   > Set isSliding(isSliding:bool=true) -> (Output_Get:bool)
   Set DynamicsAlpha(execute=@n3.else, DynamicsAlpha=1)
   > Set isSliding(isSliding=false)
   n4 = Get Character Mesh() -> (Character Mesh:Skeletal Mesh Component Object Reference)
   Set Shoulder_rot_up_r(
    execute=@n1.then_2
    Shoulder_rot_up_r:double=GetAlphaFromReferenceTransformDelta(SkeletalMeshComponent=@n4, BoneName=clavicle_r, BaseName=spine_05, DeltaComponentType=EulerY, InputRange=(X=0,Y=110), TwistAxis="0,0,1")
   ) -> (Output_Get:double)
   > Set Shoulder_rot_up_l(Shoulder_rot_up_l:double=GetAlphaFromReferenceTransformDelta(SkeletalMeshComponent=@n4, BoneName=clavicle_l, BaseName=spine_05, DeltaComponentType=EulerY, InputRange=(X=0,Y=110), TwistAxis="0,0,1")) -> (Output_Get:double)
   > Set Shoulder_rot_dn_r(Shoulder_rot_dn_r:double=GetAlphaFromReferenceTransformDelta(SkeletalMeshComponent=@n4, BoneName=upperarm_r, BaseName=clavicle_r, InputRange=(X=0,Y=-90), TwistAxis="0,0,1")) -> (Output_Get:double)
   > Set Shoulder_rot_dn_l(Shoulder_rot_dn_l:double=GetAlphaFromReferenceTransformDelta(SkeletalMeshComponent=@n4, BoneName=upperarm_l, BaseName=clavicle_l, InputRange=(X=0,Y=-90), TwistAxis="0,0,1")) -> (Output_Get:double)
   n5 = Sequence(execute=@n1.then_3)
   n6 = Sequence!(execute=@n5.then_2) -> (then_0, then_1, then_2, then_3, then_4:Exec, then_5:Exec, then_6:Exec, then_7:Exec)
   n7 = Get Body Mesh() -> (Body Mesh:Skeletal Mesh Component Object Reference)
   > Set calf_shape_back_r(calf_shape_back_r:double=GetAlphaFromReferenceTransformDelta(SkeletalMeshComponent=@n7, BoneName=thigh_twist_01_r, BaseName=pelvis, InputRange=(X=-20,Y=-120), TwistAxis="0,0,1")) -> (Output_Get:double)
   > Set calf_shape_back_l(calf_shape_back_l:double=GetAlphaFromReferenceTransformDelta(SkeletalMeshComponent=@n7, BoneName=thigh_l, BaseName=pelvis, InputRange=(X=-20,Y=-150), TwistAxis="0,0,1")) -> (Output_Get:double)
   n8 = Get Body Mesh()
   > Set lowerarm_rot_up_l(lowerarm_rot_up_l:double=GetAlphaFromReferenceTransformDelta(SkeletalMeshComponent=@n8, BoneName=lowerarm_l, BaseName=upperarm_l, InputRange=(X=-20,Y=70), TwistAxis="0,0,1")) -> (Output_Get:double)
   > Set lowerarm_rot_up_r(lowerarm_rot_up_r:double=GetAlphaFromReferenceTransformDelta(SkeletalMeshComponent=@n8, BoneName=lowerarm_r, BaseName=upperarm_r, InputRange=(X=-20,Y=70), TwistAxis="0,0,1")) -> (Output_Get:double)
   > Set upperarm_rot_dn_l(upperarm_rot_dn_l:double=GetAlphaFromReferenceTransformDelta(SkeletalMeshComponent=@n8, BoneName=spine_05, BaseName=upperarm_l, InputRange=(X=0,Y=90), TwistAxis="0,0,1")) -> (Output_Get:double)
   > Set upperarm_rot_dn_r(upperarm_rot_dn_r:double=GetAlphaFromReferenceTransformDelta(SkeletalMeshComponent=@n8, BoneName=spine_05, BaseName=upperarm_r, InputRange=(X=0,Y=90), TwistAxis="0,0,1")) -> (Output_Get:double)
   > Set Shoulder_rot_f_r(Shoulder_rot_f_r:double=GetAlphaFromReferenceTransformDelta(SkeletalMeshComponent=@n8, BoneName=upperarm_r, BaseName=clavicle_r, InputRange=(X=0,Y=110), TwistAxis="0,0,1")) -> (Output_Get:double)
   > Set Shoulder_rot_f_l(Shoulder_rot_f_l:double=GetAlphaFromReferenceTransformDelta(SkeletalMeshComponent=@n8, BoneName=upperarm_l, BaseName=clavicle_l, InputRange=(X=0,Y=110), TwistAxis="0,0,1")) -> (Output_Get:double)
   > Set Head_rot_f(Head_rot_f:double=GetAlphaFromReferenceTransformDelta(SkeletalMeshComponent=@n8, BoneName=head, BaseName=neck_02, InputRange=(X=0,Y=30), TwistAxis="0,0,1")) -> (Output_Get:double)
   n9 = Get Character Mesh()
   > Set Shoulder_rot_dn_l_0(Shoulder_rot_dn_l_0:double=GetAlphaFromReferenceTransformDelta(SkeletalMeshComponent=@n9, BoneName=spine_05, BaseName=upperarm_l, DeltaComponentType=EulerY, InputRange=(X=-20,Y=60), TwistAxis="0,0,1")) -> (Output_Get:double)
   > Set Shoulder_rot_dn_r_0(Shoulder_rot_dn_r_0:double=GetAlphaFromReferenceTransformDelta(SkeletalMeshComponent=@n9, BoneName=spine_05, BaseName=upperarm_r, DeltaComponentType=EulerY, InputRange=(X=-20,Y=60), TwistAxis="0,0,1")) -> (Output_Get:double)
   n10 = Get Character Mesh()
   Set thigh_shape_fn_l(
    execute=@n6.then_2
    thigh_shape_fn_l:double=GetAlphaFromReferenceTransformDelta(SkeletalMeshComponent=@n10, BoneName=thigh_l, BaseName=pelvis, InputRange=(X=0,Y=-110), TwistAxis="0,0,1")
   ) -> (Output_Get:double)
   > Set thigh_shape_fn_r(thigh_shape_fn_r:double=GetAlphaFromReferenceTransformDelta(SkeletalMeshComponent=@n10, BoneName=thigh_r, BaseName=pelvis, InputRange=(X=0,Y=-110), TwistAxis="0,0,1")) -> (Output_Get:double)
   > Set thigh_shape_fn_r_1(thigh_shape_fn_r_1:double=GetAlphaFromReferenceTransformDelta(SkeletalMeshComponent=@n10, BoneName=calf_r, BaseName=thigh_r, InputRange=(X=0,Y=150), TwistAxis="0,0,1")) -> (Output_Get:double)
   > Set thigh_shape(thigh_shape:bool="float < float"(A=Get thigh_shape_fn_r_1() -> (thigh_shape_fn_r_1:double), B=0.8)) -> (Output_Get:bool)
   > Set thigh_shape_fn_r_0(thigh_shape_fn_r_0:double=float - float(A=1.5, B=Get thigh_shape_fn_r() -> (thigh_shape_fn_r:double))) -> (Output_Get:double)
   n11 = Cast To FortPlayerAnimInstance(execute=@n6.then_3, Object=GetAnimInstance(self=GetOwnerSkeletalMeshComponent()))
   > Set offset_weapon(offset_weapon:double=0) -> (Output_Get:double)
   > n12 = Switch on EFortWeaponCoreAnimation!(execute, Selection:EFortWeaponCoreAnimation Enum=Get WeaponCoreAnim(self=@n11)) -> (
    Melee:Exec
    Pistol:Exec
    Shotgun:Exec
    PaperBlueprint:Exec
    Rifle:Exec
    MeleeOneHand:Exec
    MachinePistol:Exec
    RocketLauncher:Exec
    GrenadeLauncher:Exec
    GoingCommando:Exec
    AssaultRifle:Exec
    TacticalShotgun:Exec
    SniperRifle:Exec
    TrapPlacement:Exec
    ShoulderLauncher:Exec
    AbilityDecoTool:Exec
    Crossbow:Exec
    C4:Exec
    RemoteControl:Exec
    DualWield:Exec
    AR_BullPup:Exec
    AR_ForwardGrip:Exec
    MedPackPaddles:Exec
    SMG_P90:Exec
    AR_DrumGun:Exec
    Consumable_Small:Exec
    Consumable_Large:Exec
    Balloon:Exec
    MountedTurret:Exec
    CreativeTool:Exec
    ExplosiveBow:Exec
    AshtonIndigo:Exec
    AshtonChicago:Exec
    MeleeDualWield:Exec
    Unarmed:Exec
   )
   n13 = Branch(
    execute=[@n12.AR_BullPup, @n12.AssaultRifle, @n12.Pistol, @n12.Rifle, @n12.Shotgun, @n12.SniperRifle, @n12.TacticalShotgun]
    Condition=OR Boolean(A=Get bIsPlayingEmote(), B=Get bIsSkyDiving())
   )
   > Set offset_weapon(offset_weapon=0)
   Set offset_weapon(
    execute=@n13.else
    offset_weapon=GetAlphaFromReferenceTransformDelta(SkeletalMeshComponent=Get Character Mesh(), BoneName=spine_05, BaseName=upperarm_r, InputRange=(X=0,Y=120), TwistAxis="0,0,1")
   )
   n14 = Get Character Mesh()
   Set upperarm_shape_l(
    execute=@n6.then_4
    upperarm_shape_l:double=GetAlphaFromReferenceTransformDelta(SkeletalMeshComponent=@n14, BoneName=spine_05, BaseName=upperarm_l, InputRange=(X=0,Y=120), TwistAxis="0,0,1")
   ) -> (Output_Get:double)
   > Set upperarm_shape_r(upperarm_shape_r:double=GetAlphaFromReferenceTransformDelta(SkeletalMeshComponent=@n14, BoneName=spine_05, BaseName=upperarm_r, InputRange=(X=0,Y=120), TwistAxis="0,0,1")) -> (Output_Get:double)
   Set is sliding(execute=@n6.then_5, is sliding:double="To Float (Boolean)"(InBool=Get bIsSliding())) -> (Output_Get:double)
   n15 = Get bIsSkydiveFloating()
   > Set Not DiveFloatingAndDiveMode(Not DiveFloatingAndDiveMode:bool=AND Boolean(A=NOT Boolean(A=@n15), B=Get bIsSkydiveDiveMode())) -> (Output_Get:bool)
   n16 = Get thigh_shape() -> (thigh_shape:bool)
   n17 = Get bIsMoving2D()
   > Set CrouchingAndNotMovingAndNotShape(CrouchingAndNotMovingAndNotShape:bool=AND Boolean!(A=Get bIsCrouching(), B=NOT Boolean(A=@n17), C:bool=NOT Boolean(A=@n16)) -> (ReturnValue)) -> (Output_Get:bool)
   n18 = Get bIsPlayingEmote()
   n19 = Get bIsSkyDiving()
   n20 = Get bIsCrouching()
   > Set MovingAnd NotCrouchNotDivingNotEmote(MovingAnd NotCrouchNotDivingNotEmote:bool=AND Boolean!(A=Get bIsMoving2D(), B=NOT Boolean(A=@n20), C=NOT Boolean(A=@n19), D:bool=NOT Boolean(A=@n18)) -> (ReturnValue)) -> (Output_Get:bool)
   > Set DivingAndCape(DivingAndCape:bool=AND Boolean!(A=Get bIsSkyDiving(), B=Get is_WearingCape()) -> (ReturnValue)) -> (Output_Get:bool)
   > Set DivingOrDBNO(DivingOrDBNO:bool=OR Boolean(A=Get bIsSkyDiving(), B=Get bIsDBNO())) -> (Output_Get:bool)
   > Set DBNO_Alpha(DBNO_Alpha:double=SelectFloat(A=0.75, B=0.1, bPickA=Get bIsDBNO())) -> (Output_Get:double)
   n21 = NOT Boolean(A=Get isSliding())
   n22 = Get bIsCrouching()
   n23 = Get bAnimDynamicsEnabled() -> (bAnimDynamicsEnabled:bool)
   > Set Is Crouhing and jumping and Not Sliding(Is Crouhing and jumping and Not Sliding:bool=OR Boolean!(A=AND Boolean!(A=@n22, B=@n21, C:bool=@n23) -> (ReturnValue), B=Get bIsJumping(), C:bool=Get bIsFalling()) -> (ReturnValue)) -> (Output_Get:bool)
   n24 = Gravity Override State Function DEPRECATED(
    execute=@n6.then_1
    CharacterPartAnimInstance=Self-Reference()
    GameTimeInSeconds=GetGameTimeInSeconds()
    GravityOverrideParams=Get Gravity Override Params()
   )
   > Set Gravity Override Cape(Gravity Override Cape:Vector=@n24) -> (Output_Get:Vector)
   Branch(execute=@n5.then_3, Condition=Get bIsInFrontEnd())
   > Set Custom Idle is Playing(Custom Idle is Playing:float=GetCurveValue(self=GetOwnerFrontEndAnimInstance(), CurveName=Frontend_Idle_Bypass)) -> (Output_Get:float)
   n25 = Event OnCharacterCustomizationLoaded()
   > n26 = Sequence!(execute) -> (then_0, then_1, then_2, then_3)
   Set Character Mesh(execute=@n26.then_1, Character Mesh:Skeletal Mesh Component Object Reference=@n25.CustomBaseMesh) -> (Output_Get:Skeletal Mesh Component Object Reference)
   Set is_WearingCape(execute=@n26.then_3, is_WearingCape:bool=true) -> (Output_Get:bool)
   n27 = CheckForBackbling()
   > Set isWearingPackBool(isWearingPackBool:bool=OR Boolean(A=@n27, B=@n27.bIsBackblingACape, C=@n27.bIsBackblingLong)) -> (Output_Get:bool)
   > Branch(Condition=OR Boolean(A=@n27, B=@n27.bIsBackblingACape, C=@n27.bIsBackblingLong))
   > Set is_WearingCape(is_WearingCape=true)
   > n28 = Branch(Condition=NOT Boolean(A=Get bIsSkyDiving()))
   > Set is_WearingCape(is_WearingCape=true)
   Set is_WearingCape(execute=@n28.else, is_WearingCape=NOT Boolean(A=@n27.bIsBackblingLong))
   Set Body Mesh(
    execute=@n26.then_2
    Body Mesh:Skeletal Mesh Component Object Reference=Get Array Item with EFort Custom Part Type(Array=@n25.CustomMeshParts)
   ) -> (Output_Get:Skeletal Mesh Component Object Reference)
   GetSkeletalMesh!(self:Custom Character Part Object Reference=Get Array Item with EFort Custom Part Type(Array:Array of Custom Character Part Object References=@n25.CustomParts) -> (Output:Custom Character Part Object Reference)) -> (ReturnValue:Skeletal Mesh Object Reference)
   Get bIsDBNOCarrying() -> (bIsDBNOCarrying:bool)
   Get bIsDBNOCarried()
   Set is_WearingCape(is_WearingCape=false)
  ]}
]
variables [
 Body Mesh {type=SkeletalMeshComponent}
 Calf Shape Back R {type=number}
 calf_shape_back_l {type=number}
 calf_shape_back_r {type=number}
 Character Mesh {type=SkeletalMeshComponent}
 CrouchingAndNotMovingAndNotShape {type=boolean}
 Custom Idle is Playing {type=number}
 custom_cape_equipped {type=boolean}
 DBNO_Alpha {type=number}
 DivingAndCape {type=boolean}
 DivingOrDBNO {type=boolean}
 DynamicsAlpha {type=number}
 FrontEndCape {type=boolean}
 Gravity Override Cape {type=Vector}
 Gravity Override Params {
  type=GravityOverrideParamsStruct
  default {
   skydive_DivingDown {gravityOverride=<-1500 -1500 -1500>}
   skydive_GoingUp {gravityOverride=<0 0 0>}
   tactical_Sprint {gravityOverride=<0 0 -500>}
  }
 }
 GravityOverride {type=Vector}
 Head_rot_f {type=number}
 Is Crouhing and jumping and Not Sliding {type=boolean}
 is sliding {type=number}
 is_WearingCape {type=boolean}
 isSliding {type=boolean}
 isWearingPackBool {type=boolean}
 lowerarm_rot_up_l {type=number}
 lowerarm_rot_up_r {type=number}
 MovingAnd NotCrouchNotDivingNotEmote {type=boolean}
 Not DiveFloatingAndDiveMode {type=boolean}
 offset_weapon {type=number}
 Shoulder_rot_dn_l {type=number}
 Shoulder_rot_dn_l_0 {type=number}
 Shoulder_rot_dn_r {type=number}
 Shoulder_rot_dn_r_0 {type=number}
 Shoulder_rot_f_l {type=number}
 Shoulder_rot_f_r {type=number}
 Shoulder_rot_up_l {type=number}
 Shoulder_rot_up_r {type=number}
 SkelMesh {type=SkeletalMeshComponent}
 SkydiveRotation {type=Rotator, default=<rot 0 -12 0>}
 thigh_shape {type=boolean}
 thigh_shape_fn_l {type=number}
 thigh_shape_fn_r {type=number}
 thigh_shape_fn_r_0 {type=number}
 thigh_shape_fn_r_1 {type=number}
 upperarm_rot_dn_l {type=number}
 upperarm_rot_dn_r {type=number}
 upperarm_shape_l {type=number}
 upperarm_shape_r {type=number}
]
