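# Physics asset for the F_MED_RetroWheels face accessory part, simulating long back hair as center, left and right capsule chains.
# Pelvis, spine and upper-arm capsules provide the body shapes, and each chain roots on spine_05.
# Root links swing and twist freely, while lower links have angular limits, mostly tight on twist.
# Cross constraints tie the lower side chains to the center chain with twist locked, fully locking rotation to its tip.
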PhysicsAsset /BRCosmetics/Characters/Player/Female/Medium/Bodies/F_MED_RetroWheels/Meshes/Parts/F_MED_RetroWheels_FaceAcc_Physics
bodies {
 dyn_Hair_BackL_01 [{shapeType=Capsule, length=0, radius=0.5, shapeName=dyn_Hair_BackL_01_capsule, rotation=<rot 0 -90 0>}]
 dyn_Hair_BackL_02 [{
   shapeType=Capsule
   length=0
   radius=0.5
   shapeName=dyn_Hair_BackL_02_capsule
   center=<1.1194 -3.1236 1.2354>
   rotation=<rot 0 -159.6324 -60.3694>
  }]
 dyn_Hair_BackL_03 [{
   shapeType=Capsule
   length=4.1764
   radius=0.5
   shapeName=dyn_Hair_BackL_03_capsule
   center=<2.6599 -2.9025 -1.7088>
   rotation=<rot 0 -98.4022 -88.7741>
  }]
 dyn_Hair_BackL_04 [{shapeType=Capsule, length=6.7718, radius=4.704, shapeName=dyn_Hair_BackL_04_capsule, center=<2.8733 -5.8884 0>, rotation=<rot 0 9.9996 90>}]
 dyn_Hair_BackL_05 [{shapeType=Capsule, length=8.3126, radius=3.1632, shapeName=dyn_Hair_BackL_05_capsule, center=<2.8733 -5.8884 0>, rotation=<rot 0 9.9996 90>}]
 dyn_Hair_BackL_06 [{shapeType=Capsule, length=6.69, radius=4.704, shapeName=dyn_Hair_BackL_06_capsule, center=<2.8733 -5.8884 0>, rotation=<rot 0 9.9996 90>}]
 dyn_Hair_BackR_01 [{shapeType=Capsule, length=0, radius=0.5, shapeName=dyn_Hair_BackR_01_capsule, rotation=<rot 0 -90 0>}]
 dyn_Hair_BackR_02 [{
   shapeType=Capsule
   length=0
   radius=0.5
   shapeName=dyn_Hair_BackR_02_capsule
   center=<1.529 -3.3221 -0.454>
   rotation=<rot 0 30.173 -61.5402>
  }]
 dyn_Hair_BackR_03 [{
   shapeType=Capsule
   length=1.7244
   radius=0.5
   shapeName=dyn_Hair_BackR_03_capsule
   center=<0.8662 -2.7811 1.7458>
   rotation=<rot 0 97.9558 -87.3803>
  }]
 dyn_Hair_BackR_04 [{shapeType=Capsule, length=5.9594, radius=5.5165, shapeName=dyn_Hair_BackR_04_capsule, center=<2.7997 -5.4709 0>, rotation=<rot 0 9.9996 90>}]
 dyn_Hair_BackR_04a [{
   shapeType=Capsule
   length=3.7471
   radius=3.0053
   shapeName=dyn_Hair_BackR_04a_capsule
   center=<2.3485 -1.5556 0.8159>
   rotation=<rot 0 5.3735 -52.0165>
  }]
 dyn_Hair_BackR_05 [{shapeType=Capsule, length=7.7086, radius=3.7672, shapeName=dyn_Hair_BackR_05_capsule, center=<2.8404 -5.7017 0>, rotation=<rot 0 9.9996 90>}]
 dyn_Hair_BackR_06 [{shapeType=Capsule, length=8.2703, radius=3.2055, shapeName=dyn_Hair_BackR_06_capsule, center=<2.9858 -6.5264 0>, rotation=<rot 0 9.9996 90>}]
 dyn_Hair_Back_01 [{shapeType=Capsule, length=0.505, radius=0.505, shapeName=dyn_Hair_Back_01_capsule, rotation=<rot 0 -90 0>}]
 dyn_Hair_Back_02 [{
   shapeType=Capsule
   length=3.9507
   radius=0.8621
   shapeName=dyn_Hair_Back_02_capsule
   center=<4.6822 -3.6014 -0.4975>
   rotation=<rot 0 -102.2646 -71.8812>
  }]
 dyn_Hair_Back_03 [{
   shapeType=Capsule
   length=5.6966
   radius=5.2425
   shapeName=dyn_Hair_Back_03_capsule
   center=<3.2135 -1.6481 -0.2511>
   rotation=<rot 0 19.7302 -73.8569>
  }]
 dyn_Hair_Back_04 [{
   shapeType=Capsule
   length=9.0321
   radius=5.9235
   shapeName=dyn_Hair_Back_04_capsule
   center=<4.2631 -7.3347 0.0662>
   rotation=<rot -0.3062 1.9771 -86.498>
  }]
 dyn_Hair_Back_05 [{
   shapeType=Capsule
   length=9.479
   radius=4.3981
   shapeName=dyn_Hair_Back_05_capsule
   center=<5.1558 -5.2545 -0.9265>
   rotation=<rot 0 -0.0004 -100>
  }]
 dyn_Hair_Back_06 [{
   shapeType=Capsule
   length=9.1543
   radius=4.7228
   shapeName=dyn_Hair_Back_06_capsule
   center=<3.2287 -5.2545 -0.9265>
   rotation=<rot 0 -0.0004 -100>
  }]
 dyn_Hair_Back_07 [{
   shapeType=Capsule
   length=6.8434
   radius=7.0338
   shapeName=dyn_Hair_Back_07_capsule
   center=<2.8631 -3.851 -0.679>
   rotation=<rot -0.0002 -0.0003 -100>
  }]
 dyn_Hair_Back_08 [{
   shapeType=Capsule
   length=8.6563
   radius=5.2208
   shapeName=dyn_Hair_Back_08_capsule
   center=<1.3134 -1.5908 1.0035>
   rotation=<rot -0.0002 -0.0004 -60>
  }]
 pelvis [{shapeType=Capsule, length=15.0592, radius=14.9408, shapeName=pelvis_capsule}]
 spine_02 [{shapeType=Capsule, length=16.9038, radius=13.0962, shapeName=spine_02_capsule}]
 spine_04 [{shapeType=Capsule, length=19.5, radius=10.5, shapeName=spine_04_capsule, center=<-0.5867 0.561 0>}]
 spine_05 [{shapeType=Capsule, length=21, radius=9, shapeName=spine_05_capsule, center=<-0.3981 -2.0489 0>}]
 upperarm_l [{
   shapeType=Capsule
   length=23.9337
   radius=8.7967
   shapeName=upperarm_l_capsule
   center=<13.8897 0 -1.2152>
   rotation=<rot 85 -0.0015 -0.0016>
  }]
 upperarm_r [{
   shapeType=Capsule
   length=25.0436
   radius=7.6868
   shapeName=upperarm_r_capsule
   center=<-13.2276 0 1.1573>
   rotation=<rot 85 -0.0014 -0.0016>
  }]
}
constraints [
 {
  bone1Name=spine_05
  bone2Name=spine_04
  swing1LimitDegrees=45
  swing1Motion=Free
  swing2LimitDegrees=45
  swing2Motion=Free
  twistLimitDegrees=45
  twistMotion=Free
 }
 {bone1Name=dyn_Hair_Back_08, bone2Name=dyn_Hair_Back_07, swing1LimitDegrees=40, swing2LimitDegrees=5, twistLimitDegrees=5}
 {bone1Name=dyn_Hair_Back_07, bone2Name=dyn_Hair_Back_06, swing1LimitDegrees=40, swing2LimitDegrees=5, twistLimitDegrees=5}
 {bone1Name=dyn_Hair_Back_06, bone2Name=dyn_Hair_Back_05, swing1LimitDegrees=40, swing2LimitDegrees=5, twistLimitDegrees=5}
 {bone1Name=dyn_Hair_Back_05, bone2Name=dyn_Hair_Back_04, swing1LimitDegrees=40, swing2LimitDegrees=5, twistLimitDegrees=5}
 {bone1Name=dyn_Hair_Back_04, bone2Name=dyn_Hair_Back_03, swing1LimitDegrees=40, swing2LimitDegrees=5, twistLimitDegrees=5}
 {
  bone1Name=dyn_Hair_Back_03
  bone2Name=dyn_Hair_Back_02
  swing1LimitDegrees=45
  swing1Motion=Free
  swing2LimitDegrees=45
  swing2Motion=Free
  twistLimitDegrees=45
  twistMotion=Free
 }
 {
  bone1Name=dyn_Hair_Back_02
  bone2Name=dyn_Hair_Back_01
  swing1LimitDegrees=45
  swing1Motion=Free
  swing2LimitDegrees=45
  swing2Motion=Free
  twistLimitDegrees=45
  twistMotion=Free
 }
 {
  bone1Name=dyn_Hair_Back_01
  bone2Name=spine_05
  swing1LimitDegrees=45
  swing1Motion=Free
  swing2LimitDegrees=45
  swing2Motion=Free
  twistLimitDegrees=45
  twistMotion=Free
 }
 {bone1Name=dyn_Hair_BackR_06, bone2Name=dyn_Hair_BackR_05, swing1LimitDegrees=20, swing2LimitDegrees=5, twistLimitDegrees=1}
 {bone1Name=dyn_Hair_BackR_05, bone2Name=dyn_Hair_BackR_04, swing1LimitDegrees=20, swing2LimitDegrees=5, twistLimitDegrees=1}
 {bone1Name=dyn_Hair_BackL_06, bone2Name=dyn_Hair_BackL_05, swing1LimitDegrees=20, swing2LimitDegrees=5, twistLimitDegrees=1}
 {bone1Name=dyn_Hair_BackL_05, bone2Name=dyn_Hair_BackL_04, swing1LimitDegrees=20, swing2LimitDegrees=5, twistLimitDegrees=1}
 {bone1Name=dyn_Hair_BackL_04, bone2Name=dyn_Hair_BackL_03, swing1LimitDegrees=20, swing2LimitDegrees=5, twistLimitDegrees=1}
 {bone1Name=dyn_Hair_BackR_04, bone2Name=dyn_Hair_BackR_03, swing1LimitDegrees=20, swing2LimitDegrees=5, twistLimitDegrees=1}
 {
  bone1Name=dyn_Hair_BackR_03
  bone2Name=dyn_Hair_BackR_02
  swing1LimitDegrees=45
  swing1Motion=Free
  swing2LimitDegrees=45
  swing2Motion=Free
  twistLimitDegrees=45
  twistMotion=Free
 }
 {
  bone1Name=dyn_Hair_BackR_02
  bone2Name=dyn_Hair_BackR_01
  swing1LimitDegrees=45
  swing1Motion=Free
  swing2LimitDegrees=45
  swing2Motion=Free
  twistLimitDegrees=45
  twistMotion=Free
 }
 {
  bone1Name=dyn_Hair_BackR_01
  bone2Name=spine_05
  swing1LimitDegrees=45
  swing1Motion=Free
  swing2LimitDegrees=45
  swing2Motion=Free
  twistLimitDegrees=45
  twistMotion=Free
 }
 {
  bone1Name=dyn_Hair_BackL_03
  bone2Name=dyn_Hair_BackL_02
  swing1LimitDegrees=45
  swing1Motion=Free
  swing2LimitDegrees=45
  swing2Motion=Free
  twistLimitDegrees=45
  twistMotion=Free
 }
 {
  bone1Name=dyn_Hair_BackL_02
  bone2Name=dyn_Hair_BackL_01
  swing1LimitDegrees=45
  swing1Motion=Free
  swing2LimitDegrees=45
  swing2Motion=Free
  twistLimitDegrees=45
  twistMotion=Free
 }
 {
  bone1Name=dyn_Hair_BackL_01
  bone2Name=spine_05
  swing1LimitDegrees=45
  swing1Motion=Free
  swing2LimitDegrees=45
  swing2Motion=Free
  twistLimitDegrees=45
  twistMotion=Free
 }
 {
  bone1Name=dyn_Hair_BackR_05
  bone2Name=dyn_Hair_Back_05
  swing1LimitDegrees=45
  swing1Motion=Free
  swing2LimitDegrees=45
  swing2Motion=Free
  twistLimitDegrees=45
  twistMotion=Locked
 }
 {
  bone1Name=dyn_Hair_BackL_05
  bone2Name=dyn_Hair_Back_05
  swing1LimitDegrees=45
  swing1Motion=Free
  swing2LimitDegrees=45
  swing2Motion=Free
  twistLimitDegrees=45
  twistMotion=Locked
 }
 {
  bone1Name=dyn_Hair_BackR_06
  bone2Name=dyn_Hair_Back_06
  swing1LimitDegrees=45
  swing1Motion=Free
  swing2LimitDegrees=45
  swing2Motion=Free
  twistLimitDegrees=45
  twistMotion=Locked
 }
 {
  bone1Name=dyn_Hair_BackL_06
  bone2Name=dyn_Hair_Back_06
  swing1LimitDegrees=45
  swing1Motion=Free
  swing2LimitDegrees=45
  swing2Motion=Free
  twistLimitDegrees=45
  twistMotion=Locked
 }
 {bone1Name=dyn_Hair_BackR_04a, bone2Name=dyn_Hair_BackR_04, swing1LimitDegrees=45, swing2LimitDegrees=45, twistLimitDegrees=45}
 {
  bone1Name=dyn_Hair_BackR_06
  bone2Name=dyn_Hair_Back_07
  swing1LimitDegrees=45
  swing1Motion=Free
  swing2LimitDegrees=45
  swing2Motion=Free
  twistLimitDegrees=45
  twistMotion=Locked
 }
 {
  bone1Name=dyn_Hair_BackL_06
  bone2Name=dyn_Hair_Back_07
  swing1LimitDegrees=45
  swing1Motion=Free
  swing2LimitDegrees=45
  swing2Motion=Free
  twistLimitDegrees=45
  twistMotion=Locked
 }
 {
  bone1Name=dyn_Hair_BackR_06
  bone2Name=dyn_Hair_Back_08
  swing1LimitDegrees=45
  swing1Motion=Locked
  swing2LimitDegrees=45
  swing2Motion=Locked
  twistLimitDegrees=45
  twistMotion=Locked
 }
 {
  bone1Name=dyn_Hair_BackL_06
  bone2Name=dyn_Hair_Back_08
  swing1LimitDegrees=45
  swing1Motion=Locked
  swing2LimitDegrees=45
  swing2Motion=Locked
  twistLimitDegrees=45
  twistMotion=Locked
 }
]
