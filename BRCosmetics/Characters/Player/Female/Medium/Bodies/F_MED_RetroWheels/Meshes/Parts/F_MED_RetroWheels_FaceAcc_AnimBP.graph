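# Anim Blueprint for the RetroWheels female face accessory, adding secondary motion to its hair and hat bones.
# It copies the character mesh pose through a Control Rig, then runs AnimDynamics chains, with back hair simulated relative to a stabilized "gravity" bone and limited by upper-arm spheres.
# Cached poses blend on dynamics-enabled, dive, pack and DBNO states.
# Neck-rotation-driven offsets and spine constraints then correct the final pose.
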AnimBlueprint /BRCosmetics/Characters/Player/Female/Medium/Bodies/F_MED_RetroWheels/Meshes/Parts/F_MED_RetroWheels_FaceAcc_AnimBP
target_skeleton=&/F_MED_RetroWheels_FaceAcc_Skeleton
graphs [
 AnimGraph {nodes [
   n1 = OR Boolean!(A=Get bIsSkyDiving(), B=Get bIsInFrontEnd(), C:bool) -> (ReturnValue)
   n2 = "Constraint - dyn_Hair_BackR_03a (PP)"!(ComponentPose, bAlphaBoolEnabled:bool=@n1, ConstraintWeights_0=0.2, ConstraintWeights_1:float=0.8) -> (Pose) {
    alphaInputType=Bool
    alphaBoolBlend {blendOption=Linear}
    boneToModify {boneName=dyn_Hair_BackR_03a}
    constraintSetup [{targetBone {boneName=dyn_Hair_BackR_03a}}, {targetBone {boneName=clavicle_r}}]
    constraintWeights [1,1]
   }
   n3 = "Transform (Modify) Bone - Bone: dyn_Hair_BackR_03a"!(ComponentPose=@n2, bAlphaBoolEnabled:bool=True, Translation="0,0,0", Rotation="-10,0,5", Scale="1,1,1") -> (Pose) {
    alphaInputType=Bool
    rotationMode=BMM_Additive
    rotationSpace=BCS_BoneSpace
    scaleMode=BMM_Additive
    scaleSpace=BCS_BoneSpace
    translationMode=BMM_Additive
    translationSpace=BCS_BoneSpace
    alphaBoolBlend {blendOption=Linear}
    boneToModify {boneName=dyn_Hair_BackR_03a}
    rotation {pitch=-10, roll=5}
   }
   "Transform (Modify) Bone - Bone: dyn_Hair_BackR_03a"(ComponentPose=@n3, bAlphaBoolEnabled=@n1, Translation="0,0,0", Rotation="20,0,-5", Scale="1,1,1") {
    alphaInputType=Bool
    rotationMode=BMM_Additive
    rotationSpace=BCS_BoneSpace
    scaleMode=BMM_Additive
    scaleSpace=BCS_BoneSpace
    translationMode=BMM_Additive
    translationSpace=BCS_BoneSpace
    alphaBoolBlend {blendOption=Linear}
    boneToModify {boneName=dyn_Hair_BackR_03a}
    rotation {pitch=20, roll=-5}
   }
   n4 = "Constraint - dyn_Hair_BackL_03a (PP)"!(ComponentPose, bAlphaBoolEnabled:bool=@n1, ConstraintWeights_0=0.2, ConstraintWeights_1:float=0.8) -> (Pose) {
    alphaInputType=Bool
    alphaBoolBlend {blendOption=Linear}
    boneToModify {boneName=dyn_Hair_BackL_03a}
    constraintSetup [{targetBone {boneName=dyn_Hair_BackL_03a}}, {targetBone {boneName=clavicle_l}}]
    constraintWeights [1,1]
   }
   "Transform (Modify) Bone - Bone: dyn_Hair_BackL_03a"!(ComponentPose=@n4, bAlphaBoolEnabled:bool=@n1, Translation="0,0,0", Rotation="0,0,-5", Scale="1,1,1") -> (Pose) {
    alphaInputType=Bool
    rotationMode=BMM_Additive
    rotationSpace=BCS_BoneSpace
    scaleMode=BMM_Additive
    scaleSpace=BCS_BoneSpace
    translationMode=BMM_Additive
    translationSpace=BCS_BoneSpace
    alphaBoolBlend {blendOption=Linear}
    boneToModify {boneName=dyn_Hair_BackL_03a}
    rotation {roll=-5}
   }
   n5 = "Transform (Modify) Bone - Bone: dyn_Hair_BackL_03"!(ComponentPose, bAlphaBoolEnabled:bool=True, Translation="0,0,0", Rotation="0,5,0", Scale="1,1,1") -> (Pose) {
    alphaInputType=Bool
    rotationMode=BMM_Additive
    rotationSpace=BCS_BoneSpace
    scaleMode=BMM_Additive
    scaleSpace=BCS_BoneSpace
    translationMode=BMM_Additive
    translationSpace=BCS_BoneSpace
    alphaBoolBlend {blendOption=Linear}
    boneToModify {boneName=dyn_Hair_BackL_03}
    rotation {yaw=5}
   }
   "Transform (Modify) Bone - Bone: dyn_Hair_BackL_05"!(ComponentPose=@n5, bAlphaBoolEnabled:bool=True, Translation="0,0,0", Rotation="0,-5,0", Scale="1,1,1") -> (Pose) {
    alphaInputType=Bool
    rotationMode=BMM_Additive
    rotationSpace=BCS_BoneSpace
    scaleMode=BMM_Additive
    scaleSpace=BCS_BoneSpace
    translationMode=BMM_Additive
    translationSpace=BCS_BoneSpace
    alphaBoolBlend {blendOption=Linear}
    boneToModify {boneName=dyn_Hair_BackL_05}
    rotation {yaw=-5}
   }
   n6 = Blend Poses by bool(
    BlendPose_0=Two Way Blend(
     A=Use cached pose 'HairDynPose'()
     B=Use cached pose 'HairCachePose'()
     Alpha=Get IsWearingPackBool or DBNO or DBNOCarrying() -> (IsWearingPackBool or DBNO or DBNOCarrying:double)
    ) {blendNode {alphaBoolBlend {blendOption=Linear}}}
    BlendPose_1=Use cached pose 'HairDynPose'()
    bActiveValue=Get Is Dive or Dive Floating or Wearing or DBNO() -> (Is Dive or Dive Floating or Wearing or DBNO:bool)
   ) {blendType=Linear}
   n7 = Use cached pose 'HairCachePose'()
   n8 = Get bAnimDynamicsEnabled() -> (bAnimDynamicsEnabled:bool)
   n9 = Get neck_rot_up() -> (neck_rot_up:double)
   n10 = "Constraint - dyn_Hair_BackR_03a (PP)"!(
    ComponentPose=Local To Component(LocalPose=Blend Poses by bool(BlendPose_0=@n6, BlendPose_1=@n7, bActiveValue=@n8) {blendType=Linear})
    Alpha:float=@n9
    ConstraintWeights_0=0.5
    ConstraintWeights_1=0.5
   ) -> (Pose) {
    alphaBoolBlend {blendOption=Linear}
    boneToModify {boneName=dyn_Hair_BackR_03a}
    constraintSetup [{targetBone {boneName=dyn_Hair_BackR_03}}, {targetBone {boneName=dyn_Hair_BackR_04}}]
    constraintWeights [1,1]
   }
   n11 = "Transform (Modify) Bone - Bone: dyn_Hair_BackR_03a"!(ComponentPose=@n10, Alpha:float=@n9, Translation="5,0,10", Rotation="0,0,0", Scale="1,1,1") -> (Pose) {
    rotationMode=BMM_Additive
    rotationSpace=BCS_BoneSpace
    scaleMode=BMM_Additive
    scaleSpace=BCS_BoneSpace
    translationMode=BMM_Additive
    translationSpace=BCS_BoneSpace
    alphaBoolBlend {blendOption=Linear}
    boneToModify {boneName=dyn_Hair_BackR_03a}
    translation {x=5, z=10}
   }
   n12 = "Transform (Modify) Bone - Bone: dyn_Hat_R_01"(ComponentPose=@n11, Alpha=Get neck_rot_r() -> (neck_rot_r:double), Translation="0,0,0", Rotation="20,0,0", Scale="1,1,1") {
    rotationMode=BMM_Additive
    rotationSpace=BCS_BoneSpace
    scaleMode=BMM_Additive
    scaleSpace=BCS_BoneSpace
    translationMode=BMM_Additive
    translationSpace=BCS_BoneSpace
    alphaBoolBlend {blendOption=Linear}
    boneToModify {boneName=dyn_Hat_R_01}
    rotation {pitch=20}
   }
   n13 = "Transform (Modify) Bone - Bone: dyn_Hat_L_01"(ComponentPose=@n12, Alpha=Get neck_rot_l() -> (neck_rot_l:double), Translation="0,0,0", Rotation="-20,0,0", Scale="1,1,1") {
    rotationMode=BMM_Additive
    rotationSpace=BCS_BoneSpace
    scaleMode=BMM_Additive
    scaleSpace=BCS_BoneSpace
    translationMode=BMM_Additive
    translationSpace=BCS_BoneSpace
    alphaBoolBlend {blendOption=Linear}
    boneToModify {boneName=dyn_Hat_L_01}
    rotation {pitch=-20}
   }
   n14 = "Transform (Modify) Bone - Bone: dyn_Hair_Back_04"!(ComponentPose=@n13, bAlphaBoolEnabled:bool=True, Translation="0,0,0", Rotation="0,0,0", Scale="0.6,1,1") -> (Pose) {
    alphaInputType=Bool
    rotationMode=BMM_Additive
    rotationSpace=BCS_BoneSpace
    scaleMode=BMM_Additive
    scaleSpace=BCS_BoneSpace
    translationMode=BMM_Additive
    translationSpace=BCS_BoneSpace
    alphaBoolBlend {blendOption=Linear}
    boneToModify {boneName=dyn_Hair_Back_04}
    scale {x=0.6}
   }
   n15 = "Transform (Modify) Bone - Bone: dyn_Hair_BackL_04"!(ComponentPose=@n14, bAlphaBoolEnabled:bool=True, Translation="0,0,0", Rotation="0,0,0", Scale="0.6,1,1") -> (Pose) {
    alphaInputType=Bool
    rotationMode=BMM_Additive
    rotationSpace=BCS_BoneSpace
    scaleMode=BMM_Additive
    scaleSpace=BCS_BoneSpace
    translationMode=BMM_Additive
    translationSpace=BCS_BoneSpace
    alphaBoolBlend {blendOption=Linear}
    boneToModify {boneName=dyn_Hair_BackL_04}
    scale {x=0.6}
   }
   n16 = "Transform (Modify) Bone - Bone: dyn_Hair_BackR_04"!(ComponentPose=@n15, bAlphaBoolEnabled:bool=True, Translation="0,0,0", Rotation="0,0,0", Scale="0.6,1,1") -> (Pose) {
    alphaInputType=Bool
    rotationMode=BMM_Additive
    rotationSpace=BCS_BoneSpace
    scaleMode=BMM_Additive
    scaleSpace=BCS_BoneSpace
    translationMode=BMM_Additive
    translationSpace=BCS_BoneSpace
    alphaBoolBlend {blendOption=Linear}
    boneToModify {boneName=dyn_Hair_BackR_04}
    scale {x=0.6}
   }
   n17 = "Constraint - dyn_Hair_BackL_05 (PP)"!(ComponentPose=@n16, Alpha=1, ConstraintWeights_0=0.5, ConstraintWeights_1:float=0.5) -> (Pose) {
    alphaBoolBlend {blendOption=Linear}
    boneToModify {boneName=dyn_Hair_BackL_05}
    constraintSetup [{targetBone {boneName=dyn_Hair_BackL_05}}, {targetBone {boneName=spine_04}}]
    constraintWeights [1,1]
   }
   n18 = "Constraint - dyn_Hair_BackL_06 (PP)"!(ComponentPose=@n17, Alpha=1, ConstraintWeights_0=0.5, ConstraintWeights_1:float=0.5) -> (Pose) {
    alphaBoolBlend {blendOption=Linear}
    boneToModify {boneName=dyn_Hair_BackL_06}
    constraintSetup [{targetBone {boneName=dyn_Hair_BackL_06}}, {targetBone {boneName=spine_04}}]
    constraintWeights [1,1]
   }
   n19 = "Constraint - dyn_Hair_Back_07 (PP)"!(ComponentPose=@n18, Alpha=1, ConstraintWeights_0=0.5, ConstraintWeights_1:float=0.5) -> (Pose) {
    alphaBoolBlend {blendOption=Linear}
    boneToModify {boneName=dyn_Hair_Back_07}
    constraintSetup [{targetBone {boneName=spine_04}}, {targetBone {boneName=dyn_Hair_Back_06}}]
    constraintWeights [1,1]
   }
   n20 = "Constraint - dyn_Hair_Back_08 (PP)"!(ComponentPose=@n19, Alpha=1, ConstraintWeights_0=0.5, ConstraintWeights_1:float=0.5) -> (Pose) {
    alphaBoolBlend {blendOption=Linear}
    boneToModify {boneName=dyn_Hair_Back_08}
    constraintSetup [{targetBone {boneName=spine_03}}, {targetBone {boneName=dyn_Hair_Back_07}}]
    constraintWeights [1,1]
   }
   n21 = "Transform (Modify) Bone - Bone: dyn_Hair_Back_06"(ComponentPose=@n20, Alpha=1, Translation="0,2,0", Rotation="0,0,0", Scale="1,1,1") {
    rotationMode=BMM_Additive
    rotationSpace=BCS_BoneSpace
    scaleMode=BMM_Additive
    scaleSpace=BCS_BoneSpace
    translationMode=BMM_Additive
    translationSpace=BCS_BoneSpace
    alphaBoolBlend {blendOption=Linear}
    boneToModify {boneName=dyn_Hair_Back_06}
    translation {y=2}
   }
   n22 = "Transform (Modify) Bone - Bone: dyn_Hair_Back_07"(ComponentPose=@n21, Alpha=1, Translation="0,3,0", Rotation="0,0,-5", Scale="1,1,1") {
    rotationMode=BMM_Additive
    rotationSpace=BCS_BoneSpace
    scaleMode=BMM_Additive
    scaleSpace=BCS_BoneSpace
    translationMode=BMM_Additive
    translationSpace=BCS_BoneSpace
    alphaBoolBlend {blendOption=Linear}
    boneToModify {boneName=dyn_Hair_Back_07}
    rotation {roll=-5}
    translation {y=3}
   }
   Output Pose(Result=Component To Local(ComponentPose=@n22)) {layerGroup=None}
   n23 = Control Rig(Source=Copy Pose From Mesh(SourceMeshComponent=Get Character Mesh() -> (Character Mesh:Skeletal Mesh Component Object Reference)) {bUseAttachedParent=true}) {
    alphaBoolBlend {blendOption=Linear}
    controlRigAssetReference {blueprintRigClass=&/F_MED_RetroWheels_FaceAcc_CtrlRig.F_MED_RetroWheels_FaceAcc_CtrlRig_C}
   }
   n24 = "Transform (Modify) Bone - Bone: head_fix"(ComponentPose=Local To Component(LocalPose=@n23), Alpha=1, Translation="0.6,0,0", Rotation="0,0,0", Scale="1,1,1") {
    rotationMode=BMM_Additive
    rotationSpace=BCS_BoneSpace
    scaleMode=BMM_Additive
    scaleSpace=BCS_BoneSpace
    translationMode=BMM_Additive
    translationSpace=BCS_BoneSpace
    alphaBoolBlend {blendOption=Linear}
    boneToModify {boneName=head_fix}
    translation {x=0.6}
   }
   n25 = "Constraint - gravity (RR)"!(ComponentPose=@n24, Alpha=1, ConstraintWeights_0=1, ConstraintWeights_1:float=0) -> (Pose) {
    alphaBoolBlend {blendOption=Linear}
    boneToModify {boneName=gravity}
    constraintSetup [{transformType=Rotation, targetBone {boneName=root}}, {transformType=Rotation, targetBone {boneName=spine_01}}]
    constraintWeights [1,1]
   }
   n26 = "Constraint - gravity (TT)"!(ComponentPose=@n25, Alpha=1, ConstraintWeights_0=0.3, ConstraintWeights_1:float=0.7) -> (Pose) {
    alphaBoolBlend {blendOption=Linear}
    boneToModify {boneName=gravity}
    constraintSetup [{transformType=Translation, targetBone {boneName=root}}, {transformType=Translation, targetBone {boneName=spine_01}}]
    constraintWeights [1,1]
   }
   n27 = "Transform (Modify) Bone - Bone: gravity"(ComponentPose=@n26, Alpha=1, Translation="0,0,0", Rotation="0,0,0", Scale="1,1,1") {
    rotationMode=BMM_Replace
    scaleMode=BMM_Additive
    scaleSpace=BCS_BoneSpace
    translationMode=BMM_Additive
    translationSpace=BCS_BoneSpace
    alphaBoolBlend {blendOption=Linear}
    boneToModify {boneName=gravity}
   }
   n28 = Anim Dynamics - Bone: dyn_Hat_L_01(ComponentPose=@n27, Alpha=0.2, ExternalForce="0,0,0") {
    angularDampingOverride=0.99
    angularSpringConstant=75
    bAngularSpring=true
    bGravityOverrideInSimSpace=true
    bOverrideAngularDamping=true
    bUseGravityOverride=true
    simulationSpace=RootRelative
    alphaBoolBlend {blendOption=Linear}
    boundBone {boneName=dyn_Hat_L_01}
    componentAppliedLinearAccClamp=<3500 3500 3000>
    componentLinearAccScale=<0.5 0.5 0.5>
    componentLinearVelScale=<0.65 0.65 0.65>
    gravityOverride {z=-980}
    physicsBodyDefinitions [{
      collisionType=CustomSphere
      sphereCollisionRadius=2
      boundBone {boneName=dyn_Hat_L_01}
      boxExtents=<40 40 40>
      constraintSetup {angularLimitsMax=<0 15 1>, angularLimitsMin=<0 -15 -1>, angularTarget=<0 0 0>}
      localJointOffset=<-5 0.1 0.1>
     }]
    relativeSpaceBone {boneName=root}
   }
   n29 = Anim Dynamics - Bone: dyn_Hat_L_02(ComponentPose=@n28, Alpha=1, ExternalForce="0,0,0") {
    angularDampingOverride=0.99
    angularSpringConstant=75
    bAngularSpring=true
    bGravityOverrideInSimSpace=true
    bOverrideAngularDamping=true
    bUseGravityOverride=true
    simulationSpace=RootRelative
    alphaBoolBlend {blendOption=Linear}
    boundBone {boneName=dyn_Hat_L_02}
    componentAppliedLinearAccClamp=<3500 3500 3000>
    componentLinearAccScale=<0.5 0.5 0.5>
    componentLinearVelScale=<0.65 0.65 0.65>
    gravityOverride {z=-980}
    physicsBodyDefinitions [{
      collisionType=CustomSphere
      sphereCollisionRadius=2
      boundBone {boneName=dyn_Hat_L_02}
      boxExtents=<25 25 25>
      constraintSetup {angularLimitsMax=<0 15 1>, angularLimitsMin=<0 -15 -1>, angularTarget=<0 0 0>}
      localJointOffset=<-5 0.1 0.1>
     }]
    relativeSpaceBone {boneName=root}
   }
   n30 = Anim Dynamics - Bone: dyn_Hat_R_01(ComponentPose=@n29, Alpha=0.2, ExternalForce="0,0,0") {
    angularDampingOverride=0.99
    angularSpringConstant=75
    bAngularSpring=true
    bGravityOverrideInSimSpace=true
    bOverrideAngularDamping=true
    bUseGravityOverride=true
    simulationSpace=RootRelative
    alphaBoolBlend {blendOption=Linear}
    boundBone {boneName=dyn_Hat_R_01}
    componentAppliedLinearAccClamp=<3500 3500 3000>
    componentLinearAccScale=<0.5 0.5 0.5>
    componentLinearVelScale=<0.65 0.65 0.65>
    gravityOverride {z=-980}
    physicsBodyDefinitions [{
      collisionType=CustomSphere
      sphereCollisionRadius=2
      boundBone {boneName=dyn_Hat_R_01}
      boxExtents=<40 40 40>
      constraintSetup {angularLimitsMax=<0 15 1>, angularLimitsMin=<0 -15 -1>, angularTarget=<0 0 0>}
      localJointOffset=<-5 0.1 0.1>
     }]
    relativeSpaceBone {boneName=root}
   }
   n31 = Anim Dynamics - Bone: dyn_Hat_R_02(ComponentPose=@n30, Alpha=1, ExternalForce="0,0,0") {
    angularDampingOverride=0.99
    angularSpringConstant=75
    bAngularSpring=true
    bGravityOverrideInSimSpace=true
    bOverrideAngularDamping=true
    bUseGravityOverride=true
    simulationSpace=RootRelative
    alphaBoolBlend {blendOption=Linear}
    boundBone {boneName=dyn_Hat_R_02}
    componentAppliedLinearAccClamp=<3500 3500 3000>
    componentLinearAccScale=<0.5 0.5 0.5>
    componentLinearVelScale=<0.65 0.65 0.65>
    gravityOverride {z=-980}
    physicsBodyDefinitions [{
      collisionType=CustomSphere
      sphereCollisionRadius=2
      boundBone {boneName=dyn_Hat_R_02}
      boxExtents=<25 25 25>
      constraintSetup {angularLimitsMax=<0 15 1>, angularLimitsMin=<0 -15 -1>, angularTarget=<0 0 0>}
      localJointOffset=<-5 0.1 0.1>
     }]
    relativeSpaceBone {boneName=root}
   }
   n32 = Anim Dynamics - Bone: dyn_Hair_a_01(ComponentPose=@n31, Alpha=0.2, ExternalForce="0,0,0") {
    angularDampingOverride=0.99
    angularSpringConstant=75
    bAngularSpring=true
    bGravityOverrideInSimSpace=true
    bOverrideAngularDamping=true
    bUseGravityOverride=true
    simulationSpace=RootRelative
    alphaBoolBlend {blendOption=Linear}
    boundBone {boneName=dyn_Hair_a_01}
    componentAppliedLinearAccClamp=<3500 3500 3000>
    componentLinearAccScale=<0.5 0.5 0.5>
    componentLinearVelScale=<0.65 0.65 0.65>
    gravityOverride {z=-980}
    physicsBodyDefinitions [{
      collisionType=CustomSphere
      sphereCollisionRadius=2
      boundBone {boneName=dyn_Hair_a_01}
      boxExtents=<40 40 40>
      constraintSetup {angularLimitsMax=<0 15 1>, angularLimitsMin=<0 -15 -1>, angularTarget=<0 0 0>}
      localJointOffset=<-5 0.1 0.1>
     }]
    relativeSpaceBone {boneName=root}
   }
   n33 = Anim Dynamics - Bone: dyn_Hair_a_02(ComponentPose=@n32, Alpha=1, ExternalForce="0,0,0") {
    angularDampingOverride=0.99
    angularSpringConstant=75
    bAngularSpring=true
    bGravityOverrideInSimSpace=true
    bOverrideAngularDamping=true
    bUseGravityOverride=true
    simulationSpace=RootRelative
    alphaBoolBlend {blendOption=Linear}
    boundBone {boneName=dyn_Hair_a_02}
    componentAppliedLinearAccClamp=<3500 3500 3000>
    componentLinearAccScale=<0.5 0.5 0.5>
    componentLinearVelScale=<0.65 0.65 0.65>
    gravityOverride {z=-980}
    physicsBodyDefinitions [{
      collisionType=CustomSphere
      sphereCollisionRadius=2
      boundBone {boneName=dyn_Hair_a_02}
      boxExtents=<25 25 25>
      constraintSetup {angularLimitsMax=<0 15 1>, angularLimitsMin=<0 -15 -1>, angularTarget=<0 0 0>}
      localJointOffset=<-5 0.1 0.1>
     }]
    relativeSpaceBone {boneName=root}
   }
   n34 = Anim Dynamics - Bone: dyn_Hair_a_03(ComponentPose=@n33, Alpha=1, ExternalForce="0,0,0") {
    angularDampingOverride=0.99
    angularSpringConstant=75
    bAngularSpring=true
    bGravityOverrideInSimSpace=true
    bOverrideAngularDamping=true
    bUseGravityOverride=true
    simulationSpace=RootRelative
    alphaBoolBlend {blendOption=Linear}
    boundBone {boneName=dyn_Hair_a_03}
    componentAppliedLinearAccClamp=<3500 3500 3000>
    componentLinearAccScale=<0.5 0.5 0.5>
    componentLinearVelScale=<0.65 0.65 0.65>
    gravityOverride {z=-980}
    physicsBodyDefinitions [{
      collisionType=CustomSphere
      sphereCollisionRadius=2
      boundBone {boneName=dyn_Hair_a_03}
      boxExtents=<25 25 25>
      constraintSetup {angularLimitsMax=<0 15 1>, angularLimitsMin=<0 -15 -1>, angularTarget=<0 0 0>}
      localJointOffset=<-5 0.1 0.1>
     }]
    relativeSpaceBone {boneName=root}
   }
   n35 = Anim Dynamics - Bone: dyn_Hair_b_01(ComponentPose=@n34, Alpha=0.2, ExternalForce="0,0,0") {
    angularDampingOverride=0.99
    angularSpringConstant=75
    bAngularSpring=true
    bGravityOverrideInSimSpace=true
    bOverrideAngularDamping=true
    bUseGravityOverride=true
    simulationSpace=RootRelative
    alphaBoolBlend {blendOption=Linear}
    boundBone {boneName=dyn_Hair_b_01}
    componentAppliedLinearAccClamp=<3500 3500 3000>
    componentLinearAccScale=<0.5 0.5 0.5>
    componentLinearVelScale=<0.65 0.65 0.65>
    gravityOverride {z=-980}
    physicsBodyDefinitions [{
      collisionType=CustomSphere
      sphereCollisionRadius=2
      boundBone {boneName=dyn_Hair_b_01}
      boxExtents=<40 40 40>
      constraintSetup {angularLimitsMax=<0 15 1>, angularLimitsMin=<0 -15 -1>, angularTarget=<0 0 0>}
      localJointOffset=<-5 0.1 0.1>
     }]
    relativeSpaceBone {boneName=root}
   }
   n36 = Anim Dynamics - Bone: dyn_Hair_b_02(ComponentPose=@n35, Alpha=1, ExternalForce="0,0,0") {
    angularDampingOverride=0.99
    angularSpringConstant=75
    bAngularSpring=true
    bGravityOverrideInSimSpace=true
    bOverrideAngularDamping=true
    bUseGravityOverride=true
    simulationSpace=RootRelative
    alphaBoolBlend {blendOption=Linear}
    boundBone {boneName=dyn_Hair_b_02}
    componentAppliedLinearAccClamp=<3500 3500 3000>
    componentLinearAccScale=<0.5 0.5 0.5>
    componentLinearVelScale=<0.65 0.65 0.65>
    gravityOverride {z=-980}
    physicsBodyDefinitions [{
      collisionType=CustomSphere
      sphereCollisionRadius=2
      boundBone {boneName=dyn_Hair_b_02}
      boxExtents=<25 25 25>
      constraintSetup {angularLimitsMax=<0 15 1>, angularLimitsMin=<0 -15 -1>, angularTarget=<0 0 0>}
      localJointOffset=<-5 0.1 0.1>
     }]
    relativeSpaceBone {boneName=root}
   }
   n37 = Anim Dynamics - Bone: dyn_Hair_c_01(ComponentPose=@n36, Alpha=0.2, ExternalForce="0,0,0") {
    angularDampingOverride=0.99
    angularSpringConstant=75
    bAngularSpring=true
    bGravityOverrideInSimSpace=true
    bOverrideAngularDamping=true
    bUseGravityOverride=true
    simulationSpace=RootRelative
    alphaBoolBlend {blendOption=Linear}
    boundBone {boneName=dyn_Hair_c_01}
    componentAppliedLinearAccClamp=<3500 3500 3000>
    componentLinearAccScale=<0.5 0.5 0.5>
    componentLinearVelScale=<0.65 0.65 0.65>
    gravityOverride {z=-980}
    physicsBodyDefinitions [{
      collisionType=CustomSphere
      sphereCollisionRadius=2
      boundBone {boneName=dyn_Hair_c_01}
      boxExtents=<40 40 40>
      constraintSetup {angularLimitsMax=<0 15 1>, angularLimitsMin=<0 -15 -1>, angularTarget=<0 0 0>}
      localJointOffset=<-5 0.1 0.1>
     }]
    relativeSpaceBone {boneName=root}
   }
   n38 = Anim Dynamics - Bone: dyn_Hair_c_02(ComponentPose=@n37, Alpha=1, ExternalForce="0,0,0") {
    angularDampingOverride=0.99
    angularSpringConstant=75
    bAngularSpring=true
    bGravityOverrideInSimSpace=true
    bOverrideAngularDamping=true
    bUseGravityOverride=true
    simulationSpace=RootRelative
    alphaBoolBlend {blendOption=Linear}
    boundBone {boneName=dyn_Hair_c_02}
    componentAppliedLinearAccClamp=<3500 3500 3000>
    componentLinearAccScale=<0.5 0.5 0.5>
    componentLinearVelScale=<0.65 0.65 0.65>
    gravityOverride {z=-980}
    physicsBodyDefinitions [{
      collisionType=CustomSphere
      sphereCollisionRadius=2
      boundBone {boneName=dyn_Hair_c_02}
      boxExtents=<25 25 25>
      constraintSetup {angularLimitsMax=<0 15 1>, angularLimitsMin=<0 -15 -1>, angularTarget=<0 0 0>}
      localJointOffset=<-5 0.1 0.1>
     }]
    relativeSpaceBone {boneName=root}
   }
   Save cached pose 'HairCachePose'(Pose=Component To Local(ComponentPose=@n38)) {cacheName=HairCachePose}
   n39 = Anim Dynamics - Bone: dyn_Hair_BackL_02(ComponentPose=Local To Component(LocalPose=Use cached pose 'HairCachePose'()), Alpha=0.2, ExternalForce="0,0,0") {
    angularDampingOverride=0.99
    angularSpringConstant=75
    bAngularSpring=true
    bGravityOverrideInSimSpace=true
    bOverrideAngularDamping=true
    bUseGravityOverride=true
    simulationSpace=BoneRelative
    alphaBoolBlend {blendOption=Linear}
    boundBone {boneName=dyn_Hair_BackL_02}
    componentAppliedLinearAccClamp=<3500 3500 3000>
    componentLinearAccScale=<0.5 0.5 0.5>
    componentLinearVelScale=<0.65 0.65 0.65>
    gravityOverride {z=-980}
    physicsBodyDefinitions [{
      collisionType=CustomSphere
      sphereCollisionRadius=2
      boundBone {boneName=dyn_Hair_BackL_02}
      boxExtents=<40 40 40>
      constraintSetup {angularLimitsMax=<0 0 5>, angularLimitsMin=<0 0 0>, angularTarget=<0 0 0>}
      localJointOffset=<-2 -5 0>
     }]
    relativeSpaceBone {boneName=gravity}
   }
   n40 = Anim Dynamics - Bone: dyn_Hair_BackL_03(ComponentPose=@n39, Alpha=1, ExternalForce="0,0,0") {
    angularDampingOverride=0.99
    angularSpringConstant=75
    bAngularSpring=true
    bGravityOverrideInSimSpace=true
    bOverrideAngularDamping=true
    bUseGravityOverride=true
    simulationSpace=BoneRelative
    alphaBoolBlend {blendOption=Linear}
    boundBone {boneName=dyn_Hair_BackL_03}
    componentAppliedLinearAccClamp=<3500 3500 3000>
    componentLinearAccScale=<0.5 0.5 0.5>
    componentLinearVelScale=<0.65 0.65 0.65>
    gravityOverride {z=-980}
    physicsBodyDefinitions [{
      collisionType=CustomSphere
      sphereCollisionRadius=2
      boundBone {boneName=dyn_Hair_BackL_03}
      boxExtents=<25 25 25>
      constraintSetup {angularLimitsMax=<0 0 20>, angularLimitsMin=<0 0 0>, angularTarget=<0 0 0>}
      localJointOffset=<0 -5 0>
     }]
    relativeSpaceBone {boneName=gravity}
   }
   n41 = Anim Dynamics - Bone: dyn_Hair_BackL_04(ComponentPose=@n40, Alpha=1, ExternalForce="0,0,0") {
    angularDampingOverride=0.99
    angularSpringConstant=75
    bAngularSpring=true
    bGravityOverrideInSimSpace=true
    bOverrideAngularDamping=true
    bUseGravityOverride=true
    simulationSpace=BoneRelative
    alphaBoolBlend {blendOption=Linear}
    boundBone {boneName=dyn_Hair_BackL_04}
    componentAppliedLinearAccClamp=<3500 3500 3000>
    componentLinearAccScale=<0.5 0.5 0.5>
    componentLinearVelScale=<0.65 0.65 0.65>
    gravityOverride {z=-980}
    physicsBodyDefinitions [{
      collisionType=CustomSphere
      sphereCollisionRadius=2
      boundBone {boneName=dyn_Hair_BackL_04}
      boxExtents=<23 23 23>
      constraintSetup {angularLimitsMax=<15 2 30>, angularLimitsMin=<-5 -2 -1>, angularTarget=<0 0 0>}
      localJointOffset=<0 -5 0>
     }]
    relativeSpaceBone {boneName=gravity}
    sphericalLimits [
     {limitRadius=0, drivingBone {boneName=None}, sphereLocalOffset=<0 0 0>}
     {limitRadius=0, drivingBone {boneName=None}, sphereLocalOffset=<0 0 0>}
    ]
   }
   n42 = Anim Dynamics - Bone: dyn_Hair_BackL_05(ComponentPose=@n41, Alpha=1, ExternalForce="0,0,0") {
    angularDampingOverride=0.99
    angularSpringConstant=75
    bAngularSpring=true
    bGravityOverrideInSimSpace=true
    bOverrideAngularDamping=true
    bUseGravityOverride=true
    simulationSpace=BoneRelative
    alphaBoolBlend {blendOption=Linear}
    boundBone {boneName=dyn_Hair_BackL_05}
    componentAppliedLinearAccClamp=<3500 3500 3000>
    componentLinearAccScale=<0.5 0.5 0.5>
    componentLinearVelScale=<0.65 0.65 0.65>
    gravityOverride {z=-980}
    physicsBodyDefinitions [{
      collisionType=CustomSphere
      sphereCollisionRadius=2
      boundBone {boneName=dyn_Hair_BackL_05}
      boxExtents=<23 23 23>
      constraintSetup {angularLimitsMax=<15 2 30>, angularLimitsMin=<-5 -2 -1>, angularTarget=<0 0 0>}
      localJointOffset=<0 -5 0>
     }]
    relativeSpaceBone {boneName=gravity}
    sphericalLimits [
     {limitRadius=18, drivingBone {boneName=upperarm_l}, sphereLocalOffset=<5 14 0>}
     {limitRadius=0, drivingBone {boneName=None}, sphereLocalOffset=<0 0 0>}
    ]
   }
   n43 = Anim Dynamics - Bone: dyn_Hair_BackL_06(ComponentPose=@n42, Alpha=1, ExternalForce="0,0,0") {
    angularDampingOverride=0.99
    angularSpringConstant=75
    bAngularSpring=true
    bGravityOverrideInSimSpace=true
    bOverrideAngularDamping=true
    bUseGravityOverride=true
    simulationSpace=BoneRelative
    alphaBoolBlend {blendOption=Linear}
    boundBone {boneName=dyn_Hair_BackL_06}
    componentAppliedLinearAccClamp=<3500 3500 3000>
    componentLinearAccScale=<0.5 0.5 0.5>
    componentLinearVelScale=<0.65 0.65 0.65>
    gravityOverride {z=-980}
    physicsBodyDefinitions [{
      collisionType=CustomSphere
      sphereCollisionRadius=2
      boundBone {boneName=dyn_Hair_BackL_06}
      boxExtents=<23 23 23>
      constraintSetup {angularLimitsMax=<15 2 30>, angularLimitsMin=<-5 -2 -1>, angularTarget=<0 0 0>}
      localJointOffset=<0 -5 0>
     }]
    relativeSpaceBone {boneName=gravity}
    sphericalLimits [
     {limitRadius=18, drivingBone {boneName=upperarm_l}, sphereLocalOffset=<5 14 0>}
     {limitRadius=0, drivingBone {boneName=None}, sphereLocalOffset=<0 0 0>}
    ]
   }
   n44 = Anim Dynamics - Bone: dyn_Hair_BackL_03a(ComponentPose=@n43, Alpha=1, ExternalForce="0,0,0") {
    angularDampingOverride=0.99
    angularSpringConstant=75
    bAngularSpring=true
    bGravityOverrideInSimSpace=true
    bOverrideAngularDamping=true
    bUseGravityOverride=true
    simulationSpace=BoneRelative
    alphaBoolBlend {blendOption=Linear}
    boundBone {boneName=dyn_Hair_BackL_03a}
    componentAppliedLinearAccClamp=<3500 3500 3000>
    componentLinearAccScale=<0.5 0.5 0.5>
    componentLinearVelScale=<0.65 0.65 0.65>
    gravityOverride {z=-980}
    physicsBodyDefinitions [{
      collisionType=CustomSphere
      sphereCollisionRadius=2
      boundBone {boneName=dyn_Hair_BackL_03a}
      boxExtents=<25 25 25>
      constraintSetup {angularLimitsMax=<0 0 20>, angularLimitsMin=<0 0 -1>, angularTarget=<0 0 0>}
      localJointOffset=<0 -5 0>
     }]
    relativeSpaceBone {boneName=gravity}
   }
   n45 = Anim Dynamics - Bone: dyn_Hair_BackL_04a(ComponentPose=@n44, Alpha=1, ExternalForce="0,0,0") {
    angularDampingOverride=0.99
    angularSpringConstant=75
    bAngularSpring=true
    bGravityOverrideInSimSpace=true
    bOverrideAngularDamping=true
    bUseGravityOverride=true
    simulationSpace=BoneRelative
    alphaBoolBlend {blendOption=Linear}
    boundBone {boneName=dyn_Hair_BackL_04a}
    componentAppliedLinearAccClamp=<3500 3500 3000>
    componentLinearAccScale=<0.5 0.5 0.5>
    componentLinearVelScale=<0.65 0.65 0.65>
    gravityOverride {z=-980}
    physicsBodyDefinitions [{
      collisionType=CustomSphere
      sphereCollisionRadius=2
      boundBone {boneName=dyn_Hair_BackL_04a}
      boxExtents=<23 23 23>
      constraintSetup {angularLimitsMax=<15 2 30>, angularLimitsMin=<-5 -2 -1>, angularTarget=<0 0 0>}
      localJointOffset=<0 -5 0>
     }]
    relativeSpaceBone {boneName=gravity}
   }
   n46 = Anim Dynamics - Bone: dyn_Hair_Back_02(ComponentPose=@n45, Alpha=0.2, ExternalForce="0,0,0") {
    angularDampingOverride=0.99
    angularSpringConstant=75
    bAngularSpring=true
    bGravityOverrideInSimSpace=true
    bOverrideAngularDamping=true
    bUseGravityOverride=true
    simulationSpace=BoneRelative
    alphaBoolBlend {blendOption=Linear}
    boundBone {boneName=dyn_Hair_Back_02}
    componentAppliedLinearAccClamp=<3500 3500 3000>
    componentLinearAccScale=<0.5 0.5 0.5>
    componentLinearVelScale=<0.65 0.65 0.65>
    gravityOverride {z=-980}
    physicsBodyDefinitions [{
      collisionType=CustomSphere
      sphereCollisionRadius=2
      boundBone {boneName=dyn_Hair_Back_02}
      boxExtents=<40 40 40>
      constraintSetup {angularLimitsMax=<0 0 5>, angularLimitsMin=<0 0 0>, angularTarget=<0 0 0>}
      localJointOffset=<-2 -5 0>
     }]
    relativeSpaceBone {boneName=gravity}
   }
   n47 = Anim Dynamics - Bone: dyn_Hair_Back_03(ComponentPose=@n46, Alpha=1, ExternalForce="0,0,0") {
    angularDampingOverride=0.99
    angularSpringConstant=75
    bAngularSpring=true
    bGravityOverrideInSimSpace=true
    bOverrideAngularDamping=true
    bUseGravityOverride=true
    simulationSpace=BoneRelative
    alphaBoolBlend {blendOption=Linear}
    boundBone {boneName=dyn_Hair_Back_03}
    componentAppliedLinearAccClamp=<3500 3500 3000>
    componentLinearAccScale=<0.5 0.5 0.5>
    componentLinearVelScale=<0.65 0.65 0.65>
    gravityOverride {z=-980}
    physicsBodyDefinitions [{
      collisionType=CustomSphere
      sphereCollisionRadius=2
      boundBone {boneName=dyn_Hair_Back_03}
      boxExtents=<25 25 25>
      constraintSetup {angularLimitsMax=<0 0 20>, angularLimitsMin=<0 0 0>, angularTarget=<0 0 0>}
      localJointOffset=<0 -5 0>
     }]
    relativeSpaceBone {boneName=gravity}
   }
   n48 = Anim Dynamics - Bone: dyn_Hair_Back_04(ComponentPose=@n47, Alpha=1, ExternalForce="0,0,0") {
    angularDampingOverride=0.99
    angularSpringConstant=75
    bAngularSpring=true
    bGravityOverrideInSimSpace=true
    bOverrideAngularDamping=true
    bUseGravityOverride=true
    simulationSpace=BoneRelative
    alphaBoolBlend {blendOption=Linear}
    boundBone {boneName=dyn_Hair_Back_04}
    componentAppliedLinearAccClamp=<3500 3500 3000>
    componentLinearAccScale=<0.5 0.5 0.5>
    componentLinearVelScale=<0.65 0.65 0.65>
    gravityOverride {z=-980}
    physicsBodyDefinitions [{
      collisionType=CustomSphere
      sphereCollisionRadius=2
      boundBone {boneName=dyn_Hair_Back_04}
      boxExtents=<23 23 23>
      constraintSetup {angularLimitsMax=<2 2 30>, angularLimitsMin=<-2 -2 -1>, angularTarget=<0 0 0>}
      localJointOffset=<0 -5 0>
     }]
    relativeSpaceBone {boneName=gravity}
   }
   n49 = Anim Dynamics - Bone: dyn_Hair_Back_05(ComponentPose=@n48, Alpha=1, ExternalForce="0,0,0") {
    angularDampingOverride=0.99
    angularSpringConstant=75
    bAngularSpring=true
    bGravityOverrideInSimSpace=true
    bOverrideAngularDamping=true
    bUseGravityOverride=true
    simulationSpace=BoneRelative
    alphaBoolBlend {blendOption=Linear}
    boundBone {boneName=dyn_Hair_Back_05}
    componentAppliedLinearAccClamp=<3500 3500 3000>
    componentLinearAccScale=<0.5 0.5 0.5>
    componentLinearVelScale=<0.65 0.65 0.65>
    gravityOverride {z=-980}
    physicsBodyDefinitions [{
      collisionType=CustomSphere
      sphereCollisionRadius=2
      boundBone {boneName=dyn_Hair_Back_05}
      boxExtents=<23 23 23>
      constraintSetup {angularLimitsMax=<3 2 30>, angularLimitsMin=<-3 -2 -1>, angularTarget=<0 0 0>}
      localJointOffset=<0 -5 0>
     }]
    relativeSpaceBone {boneName=gravity}
    sphericalLimits [
     {limitRadius=0, drivingBone {boneName=None}, sphereLocalOffset=<0 0 0>}
     {limitRadius=0, drivingBone {boneName=None}, sphereLocalOffset=<0 0 0>}
    ]
   }
   n50 = Anim Dynamics - Bone: dyn_Hair_Back_06(ComponentPose=@n49, Alpha=1, ExternalForce="0,0,0") {
    angularDampingOverride=0.99
    angularSpringConstant=75
    bAngularSpring=true
    bGravityOverrideInSimSpace=true
    bOverrideAngularDamping=true
    bUseGravityOverride=true
    simulationSpace=BoneRelative
    alphaBoolBlend {blendOption=Linear}
    boundBone {boneName=dyn_Hair_Back_06}
    componentAppliedLinearAccClamp=<3500 3500 3000>
    componentLinearAccScale=<0.5 0.5 0.5>
    componentLinearVelScale=<0.65 0.65 0.65>
    gravityOverride {z=-980}
    physicsBodyDefinitions [{
      collisionType=CustomSphere
      sphereCollisionRadius=2
      boundBone {boneName=dyn_Hair_Back_06}
      boxExtents=<23 23 23>
      constraintSetup {angularLimitsMax=<4 2 30>, angularLimitsMin=<-4 -2 -1>, angularTarget=<0 0 0>}
      localJointOffset=<0 -5 0>
     }]
    relativeSpaceBone {boneName=gravity}
    sphericalLimits [
     {limitRadius=18, drivingBone {boneName=upperarm_r}, sphereLocalOffset=<5 -14 0>}
     {limitRadius=18, drivingBone {boneName=upperarm_l}, sphereLocalOffset=<5 14 0>}
    ]
   }
   n51 = Anim Dynamics - Bone: dyn_Hair_Back_07(ComponentPose=@n50, Alpha=1, ExternalForce="0,0,0") {
    angularDampingOverride=0.99
    angularSpringConstant=75
    bAngularSpring=true
    bGravityOverrideInSimSpace=true
    bOverrideAngularDamping=true
    bUseGravityOverride=true
    simulationSpace=BoneRelative
    alphaBoolBlend {blendOption=Linear}
    boundBone {boneName=dyn_Hair_Back_07}
    componentAppliedLinearAccClamp=<3500 3500 3000>
    componentLinearAccScale=<0.5 0.5 0.5>
    componentLinearVelScale=<0.65 0.65 0.65>
    gravityOverride {z=-980}
    physicsBodyDefinitions [{
      collisionType=CustomSphere
      sphereCollisionRadius=2
      boundBone {boneName=dyn_Hair_Back_07}
      boxExtents=<20 20 20>
      constraintSetup {angularLimitsMax=<4 2 30>, angularLimitsMin=<-4 -2 -1>, angularTarget=<0 0 0>}
      localJointOffset=<0 -5 0>
     }]
    relativeSpaceBone {boneName=gravity}
   }
   n52 = Anim Dynamics - Bone: dyn_Hair_Back_08(ComponentPose=@n51, Alpha=1, ExternalForce="0,0,0") {
    angularDampingOverride=0.99
    angularSpringConstant=75
    bAngularSpring=true
    bGravityOverrideInSimSpace=true
    bOverrideAngularDamping=true
    bUseGravityOverride=true
    simulationSpace=BoneRelative
    alphaBoolBlend {blendOption=Linear}
    boundBone {boneName=dyn_Hair_Back_08}
    componentAppliedLinearAccClamp=<3500 3500 3000>
    componentLinearAccScale=<0.5 0.5 0.5>
    componentLinearVelScale=<0.65 0.65 0.65>
    gravityOverride {z=-980}
    physicsBodyDefinitions [{
      collisionType=CustomSphere
      sphereCollisionRadius=2
      boundBone {boneName=dyn_Hair_Back_08}
      boxExtents=<17 17 17>
      constraintSetup {angularLimitsMax=<4 2 30>, angularLimitsMin=<-4 -2 -1>, angularTarget=<0 0 0>}
      localJointOffset=<0 -5 0>
     }]
    relativeSpaceBone {boneName=gravity}
   }
   n53 = Anim Dynamics - Bone: dyn_Hair_Back_04a(ComponentPose=@n52, Alpha=1, ExternalForce="0,0,0") {
    angularDampingOverride=0.99
    angularSpringConstant=75
    bAngularSpring=true
    bGravityOverrideInSimSpace=true
    bOverrideAngularDamping=true
    bUseGravityOverride=true
    simulationSpace=BoneRelative
    alphaBoolBlend {blendOption=Linear}
    boundBone {boneName=dyn_Hair_Back_04a}
    componentAppliedLinearAccClamp=<3500 3500 3000>
    componentLinearAccScale=<0.5 0.5 0.5>
    componentLinearVelScale=<0.65 0.65 0.65>
    gravityOverride {z=-980}
    physicsBodyDefinitions [{
      collisionType=CustomSphere
      sphereCollisionRadius=2
      boundBone {boneName=dyn_Hair_Back_04a}
      boxExtents=<23 23 23>
      constraintSetup {angularLimitsMax=<2 2 30>, angularLimitsMin=<-2 -2 -1>, angularTarget=<0 0 0>}
      localJointOffset=<0 -5 0>
     }]
    relativeSpaceBone {boneName=gravity}
   }
   n54 = Anim Dynamics - Bone: dyn_Hair_BackR_02(ComponentPose=@n53, Alpha=0.2, ExternalForce="0,0,0") {
    angularDampingOverride=0.99
    angularSpringConstant=75
    bAngularSpring=true
    bGravityOverrideInSimSpace=true
    bOverrideAngularDamping=true
    bUseGravityOverride=true
    simulationSpace=BoneRelative
    alphaBoolBlend {blendOption=Linear}
    boundBone {boneName=dyn_Hair_BackR_02}
    componentAppliedLinearAccClamp=<3500 3500 3500>
    componentLinearAccScale=<0.5 0.5 0.5>
    componentLinearVelScale=<0.65 0.65 0.65>
    gravityOverride {z=-980}
    physicsBodyDefinitions [{
      collisionType=CustomSphere
      sphereCollisionRadius=2
      boundBone {boneName=dyn_Hair_BackR_02}
      boxExtents=<40 40 40>
      constraintSetup {angularLimitsMax=<0 0 5>, angularLimitsMin=<0 0 0>, angularTarget=<0 0 0>}
      localJointOffset=<-2 -5 0>
     }]
    relativeSpaceBone {boneName=gravity}
   }
   n55 = Anim Dynamics - Bone: dyn_Hair_BackR_03(ComponentPose=@n54, Alpha=1, ExternalForce="0,0,0") {
    angularDampingOverride=0.99
    angularSpringConstant=75
    bAngularSpring=true
    bGravityOverrideInSimSpace=true
    bOverrideAngularDamping=true
    bUseGravityOverride=true
    simulationSpace=BoneRelative
    alphaBoolBlend {blendOption=Linear}
    boundBone {boneName=dyn_Hair_BackR_03}
    componentAppliedLinearAccClamp=<3500 3500 3000>
    componentLinearAccScale=<0.5 0.5 0.5>
    componentLinearVelScale=<0.65 0.65 0.65>
    gravityOverride {z=-980}
    physicsBodyDefinitions [{
      collisionType=CustomSphere
      sphereCollisionRadius=2
      boundBone {boneName=dyn_Hair_BackR_03}
      boxExtents=<25 25 25>
      constraintSetup {angularLimitsMax=<0 0 20>, angularLimitsMin=<0 0 0>, angularTarget=<0 0 0>}
      localJointOffset=<0 -5 0>
     }]
    relativeSpaceBone {boneName=gravity}
    sphericalLimits [{limitRadius=18, drivingBone {boneName=None}, sphereLocalOffset=<-5 -13 0>}]
   }
   n56 = Anim Dynamics - Bone: dyn_Hair_BackR_04(ComponentPose=@n55, Alpha=1, ExternalForce="0,0,0") {
    angularDampingOverride=0.99
    angularSpringConstant=75
    bAngularSpring=true
    bGravityOverrideInSimSpace=true
    bOverrideAngularDamping=true
    bUseGravityOverride=true
    simulationSpace=BoneRelative
    alphaBoolBlend {blendOption=Linear}
    boundBone {boneName=dyn_Hair_BackR_04}
    componentAppliedLinearAccClamp=<3500 3500 3000>
    componentLinearAccScale=<0.5 0.5 0.5>
    componentLinearVelScale=<0.65 0.65 0.65>
    gravityOverride {z=-980}
    physicsBodyDefinitions [{
      collisionType=CustomSphere
      sphereCollisionRadius=2
      boundBone {boneName=dyn_Hair_BackR_04}
      boxExtents=<23 23 23>
      constraintSetup {angularLimitsMax=<5 2 30>, angularLimitsMin=<-15 -2 -1>, angularTarget=<0 0 0>}
      localJointOffset=<0 -5 0>
     }]
    relativeSpaceBone {boneName=gravity}
   }
   n57 = Anim Dynamics - Bone: dyn_Hair_BackR_05(ComponentPose=@n56, Alpha=1, ExternalForce="0,0,0") {
    angularDampingOverride=0.99
    angularSpringConstant=75
    bAngularSpring=true
    bGravityOverrideInSimSpace=true
    bOverrideAngularDamping=true
    bUseGravityOverride=true
    simulationSpace=BoneRelative
    alphaBoolBlend {blendOption=Linear}
    boundBone {boneName=dyn_Hair_BackR_05}
    componentAppliedLinearAccClamp=<3500 3500 3000>
    componentLinearAccScale=<0.5 0.5 0.5>
    componentLinearVelScale=<0.65 0.65 0.65>
    gravityOverride {z=-980}
    physicsBodyDefinitions [{
      collisionType=CustomSphere
      sphereCollisionRadius=2
      boundBone {boneName=dyn_Hair_BackR_05}
      boxExtents=<23 23 23>
      constraintSetup {angularLimitsMax=<5 0 30>, angularLimitsMin=<-15 0 -1>, angularTarget=<0 0 0>}
      localJointOffset=<0 -5 0>
     }]
    relativeSpaceBone {boneName=gravity}
    sphericalLimits [
     {limitRadius=18, drivingBone {boneName=upperarm_r}, sphereLocalOffset=<-10 -12 0>}
     {limitRadius=0, drivingBone {boneName=None}, sphereLocalOffset=<0 0 0>}
    ]
   }
   n58 = Anim Dynamics - Bone: dyn_Hair_BackR_06(ComponentPose=@n57, Alpha=1, ExternalForce="0,0,0") {
    angularDampingOverride=0.99
    angularSpringConstant=75
    bAngularSpring=true
    bGravityOverrideInSimSpace=true
    bOverrideAngularDamping=true
    bUseGravityOverride=true
    simulationSpace=BoneRelative
    alphaBoolBlend {blendOption=Linear}
    boundBone {boneName=dyn_Hair_BackR_06}
    componentAppliedLinearAccClamp=<3500 3500 3000>
    componentLinearAccScale=<0.5 0.5 0.5>
    componentLinearVelScale=<0.65 0.65 0.65>
    gravityOverride {z=-980}
    physicsBodyDefinitions [{
      collisionType=CustomSphere
      sphereCollisionRadius=2
      boundBone {boneName=dyn_Hair_BackR_06}
      boxExtents=<23 23 23>
      constraintSetup {angularLimitsMax=<5 0 30>, angularLimitsMin=<-15 0 -1>, angularTarget=<0 0 0>}
      localJointOffset=<0 -5 0>
     }]
    relativeSpaceBone {boneName=gravity}
    sphericalLimits [
     {limitRadius=18, drivingBone {boneName=upperarm_r}, sphereLocalOffset=<-10 -12 0>}
     {limitRadius=0, drivingBone {boneName=None}, sphereLocalOffset=<0 0 0>}
    ]
   }
   n59 = RigidBody!(
    ComponentPose=@n58
    Alpha
    OverrideWorldGravity:Vector="0,0,-980"
    ExternalForce
    ComponentLinearAccScale:Vector="0.5,0.5,0.5"
    ComponentLinearVelScale:Vector="0.8,0.8,0.8"
   ) -> (Pose) {
    bClampLinearTranslationLimitToRefPose=true
    bOverrideWorldGravity=true
    overridePhysicsAsset=&/F_MED_RetroWheels_FaceAcc_Physics
    simulationSpace=BaseBoneSpace
    alphaBoolBlend {blendOption=Linear}
    baseBoneRef {boneName=gravity}
    componentAppliedLinearAccClamp=<3500 3500 3500>
    componentLinearAccScale=<0.1 0.1 0.1>
    componentLinearVelScale=<0.1 0.1 0.1>
    overrideWorldGravity {z=-980}
    simSpaceSettings {
     maxAngularAcceleration=1000
     maxAngularVelocity=0.3
     maxLinearAcceleration=700
     maxLinearVelocity=500
     velocityScaleZ=0
     worldAlpha=0.1
     externalLinearDragV=<0.5 0.5>
    }
   }
   Save cached pose 'HairDynPose'(Pose=Component To Local(ComponentPose=@n59)) {cacheName=HairDynPose}
  ]}
 EventGraph {nodes [
   Event BlueprintUpdateAnimation()
   > n1 = Sequence!(execute) -> (then_0, then_1, then_2:Exec, then_3:Exec)
   > n2 = Gravity Override State Function DEPRECATED(CharacterPartAnimInstance=Self-Reference(), GameTimeInSeconds=GetGameTimeInSeconds(), GravityOverrideParams=Get Gravity Override Params())
   > Set Gravity Override Cape(Gravity Override Cape:Vector=@n2) -> (Output_Get:Vector)
   n3 = Get bIsDBNOCarrying() -> (bIsDBNOCarrying:bool)
   n4 = Get isWearingPackBool() -> (isWearingPackBool:bool)
   n5 = Get bIsDBNO()
   Set IsWearingPackBool or DBNO or DBNOCarrying(execute=@n1.then_2, IsWearingPackBool or DBNO or DBNOCarrying:double=SelectFloat(A=0.9, B=0, bPickA=OR Boolean(A=@n4, B=@n5, C=@n3))) -> (Output_Get:double)
   Set Is Dive or Dive Floating or Wearing or DBNO(
    execute=@n1.then_3
    Is Dive or Dive Floating or Wearing or DBNO:bool=OR Boolean!(A=Get bIsSkydiveDiveMode(), B=Get bIsSkydiveFloating(), C=Get isWearingPackBool(), D:bool=Get bIsDBNO(), E:bool=Get bIsDBNOCarrying()) -> (ReturnValue)
   ) -> (Output_Get:bool)
   Set neck_rot_up(
    execute=@n1.then_1
    neck_rot_up:double=GetAlphaFromReferenceTransformDelta(
     SkeletalMeshComponent=Get Head_Mesh() -> (Head_Mesh:Skeletal Mesh Component Object Reference)
     BoneName=head
     BaseName=spine_05
     InputRange=(X=-20,Y=-80)
     TwistAxis="0,0,1"
    )
   ) -> (Output_Get:double)
   > Set neck_rot_r(neck_rot_r:double=GetAlphaFromReferenceTransformDelta(SkeletalMeshComponent=Get Head_Mesh(), BoneName=head, BaseName=spine_05, InputRange=(X=10,Y=20), TwistAxis="0,0,1")) -> (Output_Get:double)
   > Set neck_rot_l(neck_rot_l:double=GetAlphaFromReferenceTransformDelta(SkeletalMeshComponent=Get Head_Mesh(), BoneName=head, BaseName=spine_05, InputRange=(X=0,Y=20), TwistAxis="0,0,1")) -> (Output_Get:double)
   n6 = Event OnCharacterCustomizationLoaded()
   > n7 = Sequence()
   Set Character Mesh(execute=@n7.then_1, Character Mesh:Skeletal Mesh Component Object Reference=@n6.CustomBaseMesh) -> (Output_Get:Skeletal Mesh Component Object Reference)
   n8 = CheckForBackbling()
   Set isWearingPackBool(execute=@n7.then_3, isWearingPackBool:bool=OR Boolean!(A=@n8, B=@n8.bIsBackblingACape, C=@n8.bIsBackblingLong) -> (ReturnValue)) -> (Output_Get:bool)
   > Set is_WearingCape(is_WearingCape:bool=@n8.bIsBackblingLong) -> (Output_Get:bool)
   Set Head_Mesh(
    execute=@n7.then_2
    Head_Mesh:Skeletal Mesh Component Object Reference=Get Array Item with EFort Custom Part Type(Array=@n6.CustomMeshParts, PartType=Head)
   ) -> (Output_Get:Skeletal Mesh Component Object Reference)
   GetSkeletalMesh!(self:Custom Character Part Object Reference=Get Array Item with EFort Custom Part Type(Array:Array of Custom Character Part Object References=@n6.CustomParts, PartType=Head) -> (Output:Custom Character Part Object Reference)) -> (ReturnValue:Skeletal Mesh Object Reference)
   AND Boolean(A=@n8, B=NOT Boolean(A=@n8.bIsBackblingLong))
   Set Body Mesh(Body Mesh:Skeletal Mesh Component Object Reference) -> (Output_Get:Skeletal Mesh Component Object Reference)
  ]}
]
variables [
 Body Mesh {type=SkeletalMeshComponent}
 Character Mesh {type=SkeletalMeshComponent}
 Gravity Override Cape {type=Vector}
 Gravity Override Params {
  type=GravityOverrideParamsStruct
  default {
   onGround_Standing {windAmplitude=<-10 -150 0>, windFrequency=<-7 -1 -1>, windOffset=<-10 -100 0>}
   skydive_DivingDown {gravityOverride=<-500 0 0>}
   skydive_FloatingDown {gravityOverride=<0 -400 100>}
   skydive_GoingUp {gravityOverride=<0 -300 -200>}
   skydive_Parachuting {gravityOverride=<0 -600 -980>}
  }
 }
 Head_Mesh {type=SkeletalMeshComponent}
 Is Dive or Dive Floating or Wearing or DBNO {type=boolean}
 is_WearingCape {type=boolean}
 isWearingPackBool {type=boolean}
 IsWearingPackBool or DBNO or DBNOCarrying {type=number}
 neck_rot_l {type=number}
 neck_rot_r {type=number}
 neck_rot_up {type=number}
 Skel_Mesh {type=SkeletalMeshComponent}
 SkelMesh {type=SkeletalMeshComponent}
]
